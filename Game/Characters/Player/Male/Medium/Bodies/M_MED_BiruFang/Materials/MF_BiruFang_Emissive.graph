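# Material function for the BiruFang player body that replaces incoming emissive with an animated glow, optionally limited by a mask.
# Builds pivot-relative UVs from pre-skinned position, keeping projection independent of animation.
# Defines time-driven sine waves, radially panned smoke noise, a scrolling normal and a UV circle head mask as named reroutes, plus a combined effect.
# Final glow supports an optional core-color tint and brightens at night via Scalar_Time_of_Day.
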MaterialFunction /Game/Characters/Player/Male/Medium/Bodies/M_MED_BiruFang/Materials/MF_BiruFang_Emissive
graph {nodes [
  n1 = NamedRerouteUsage()
  n2 = VectorParameter() {group=MF_BF_Emissive, parameterName=BF_EmissiveCoreColor, sortPriority=33, defaultValue=<rgba 1 1 1 1>}
  n3 = ComponentMask(None=NamedRerouteUsage()) {b=true}
  n4 = LinearInterpolate(
   A=Multiply(A=NamedRerouteUsage()) {constB=2}
   B=@n2.RGB
   Alpha=LinearInterpolate(Alpha=Multiply(A=@n3) {constB=1.1}) {constA=-0.2, constB=1.2}
  )
  n5 = Divide(A=Add(A=NamedRerouteUsage()) {constB=1.4})
  n6 = Constant() {r=1}
  n7 = Multiply(
   A=StaticSwitchParameter(True=Multiply(A=@n4, B=@n5), False=NamedRerouteUsage()) {group=MF_BF_Emissive, parameterName=BF_ChanceEmissiveCoreColor, sortPriority=33}
   B=MaterialFunctionCall(Value_Morning=@n6, Value_Day=@n6, Value_Evening=@n6, Value_Night=Constant() {r=1.3}) {materialFunction=&Scalar_Time_of_Day}.Result
  )
  n8 = StaticSwitchParameter(True=NamedRerouteUsage(), False=Constant() {r=1}) {group=MF_BF_Emissive, parameterName=BF_UseEmissiveMask}
  FunctionOutput(None=SetMaterialAttributes!(MaterialAttributes=NamedRerouteUsage(), EmissiveColor=LinearInterpolate(A=@n1, B=@n7, Alpha=@n8)) {attributeSetTypes [#1]})
  n9 = Distance(A=NamedRerouteUsage(), B=Constant())
  n10 = Multiply(A=Time() {period=5}) {constB=-0.1}
  n11 = Sine(None=Multiply(A=Add(A=@n9, B=@n10)) {constB=2})
  n12 = Saturate(None=Add(A=@n11) {constB=0.5})
  NamedRerouteDeclaration(None=Add(A=Divide(A=@n12)) {constB=0.7}) {name=SineWave, nodeColor=<rgba 0 0.4141 1 1>}
  n13 = GetMaterialAttributes(None=FunctionInput() {inputType=FunctionInput_MaterialAttributes}) {attributeGetTypes [#1]}
  NamedRerouteDeclaration(None=@n13.EmissiveColor) {name=InEmissive, nodeColor=<rgba 0 0.9766 1 1>}
  n14 = ComponentMask(None=NamedRerouteUsage()) {b=true, r=true}
  n15 = NamedRerouteUsage()
  n16 = Time() {period=5}
  n17 = MaterialFunctionCall!(
   UV=AppendVector(A=ComponentMask(None=@n15) {r=true}, B=ComponentMask(None=@n15) {b=true})
   UV Tilling=AppendVector(A=Constant() {r=10}, B=Constant() {r=3})
   PannerValue=AppendVector(A=Multiply(A=@n16) {constB=-0.3}, B=Constant())
  ) {materialFunction=&/MF_RadialUVPanner}
  n18 = TextureSample!(UVs=Add(A=Multiply(A=@n14) {constB=0.5}, B=@n17.UV), Tex, "DDX(UVs)"=@n17.DDX, "DDY(UVs)"=@n17.DDX, Apply View MipBias) {
   mipValueMode=TMVM_Derivative
   samplerType=SAMPLERTYPE_LinearColor
   texture=/Game/Characters/Player/Male/Medium/Bodies/M_MED_BiruFang/Textures/T_BiruFang_Noise
  }
  NamedRerouteDeclaration(None=@n18.RGB) {name=Smoke Effect, nodeColor=<rgba 1 0.5859 0 1>}
  n19 = NamedRerouteUsage()
  n20 = Time() {period=5}
  n21 = TextureSample!(UVs=Add(A=ComponentMask(None=@n19) {b=true, r=true}, B=Multiply(A=@n20) {constB=0.02}), Tex, Apply View MipBias) {samplerType=SAMPLERTYPE_Normal, texture=/Game/Characters/Player/Male/Medium/Bodies/M_MED_BiruFang/Textures/T_BiruFang_EmissiveEffect_N}
  NamedRerouteDeclaration(None=@n21.RGB) {name=normal, nodeColor=<rgba 0.0877 0 1 1>}
  NamedRerouteDeclaration(None=@n13.MaterialAttributes) {name=Material, nodeColor=<rgba 0.5937 1 0 1>}
  n22 = Add(A=PreSkinnedPosition(), B=VectorParameter() {group=MF_BF_Emissive, parameterName=BF_Pvt_Position, defaultValue=<rgba 0.5 0 -126 1>}.RGB) {constB=15}
  NamedRerouteDeclaration(None=VertexInterpolator(VS=Multiply(A=@n22) {constB=0.02}).PS) {name=pivotUV, nodeColor=<rgba 0.4062 0 1 1>}
  NamedRerouteDeclaration(None=Divide(A=Add(A=@n11))) {name=SineWave2, nodeColor=<rgba 1 0 0.2344 1>}
  n23 = TextureCoordinate()
  n24 = Add(A=ScalarParameter() {group=MF_BF_Emissive, parameterName=CircleMaskPositionX}, B=@n23) {constB=-0.015}
  n25 = Add(A=@n23, B=ScalarParameter() {group=MF_BF_Emissive, parameterName=CircleMaskPositionZ}) {constB=-0.015}
  n26 = Distance(A=AppendVector(A=ComponentMask(None=@n24) {r=true}, B=ComponentMask(None=@n25) {g=true}), B=Constant() {r=0.5})
  NamedRerouteDeclaration(None=Step(X=OneMinus(None=@n26)) {constY=0.9}) {name=HeadMask, nodeColor=<rgba 0 0.8125 1 1>}
  n27 = NamedRerouteUsage()
  n28 = NamedRerouteUsage()
  n29 = NamedRerouteUsage()
  n30 = NamedRerouteUsage()
  n31 = Add(A=Multiply(A=@n29) {constB=0.4}, B=Multiply(A=Clamp(None=@n30) {maxDefault=0.3}) {constB=1.3})
  n32 = Subtract(A=Saturate(None=Subtract(A=@n28, B=@n29)), B=Subtract(A=Saturate(None=@n31)))
  NamedRerouteDeclaration(None=Multiply(A=Multiply(A=@n27, B=@n32), B=NamedRerouteUsage())) {name=EmissiveFx, nodeColor=<rgba 0.2187 1 0 1>}
 ]}
