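# Body material for the BiruFang male medium player character, using material attributes and supporting skeletal meshes, morph targets and Nanite.
# The shared CharacterParent_2023 PreFX and PostFX functions bracket a VFX composite that adds BiruFang emissive and a crystal function borrowed from the RoyalDusk backpack.
# The Tech Art Mask scaled by BF_MaskMultiply drives the crystal function's alpha.
# A shared FX emissive adjustment function runs last.
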Material /Game/Characters/Player/Male/Medium/Bodies/M_MED_BiruFang/Materials/M_BiruFang_Body
graph {
 properties {
  bUseMaterialAttributes=true
  bUsedWithMeshDeformer=true
  bUsedWithMorphTargets=true
  bUsedWithNanite=true
  bUsedWithSkeletalMesh=true
  pixelDepthOffsetMode=PDOM_Legacy
  usedShadingModels=Default Lit
 }
 nodes [
  n1 = MaterialFunctionCall() {materialFunction=&MF_CharacterParent_2023_PreFX}
  Composite!("In (MA)"=@n1.Result, FX=@n1.Result) {subgraphName=VFX}
  MaterialExpressionReroute_0 = Reroute(None=@n1.Result)
  MaterialExpressionReroute_2 = Reroute(None=@n1.Tech Art Mask)
  n2 = MaterialFunctionCall!(
   Material=MaterialFunctionCall!(In=@MaterialExpressionReroute_0."In (MA)") {materialFunction=&/MF_BiruFang_Emissive}.Result
   Alpha=Multiply(A=@MaterialExpressionReroute_2.FX, B=ScalarParameter() {parameterName=BF_MaskMultiply})
  ) {materialFunction=/Game/Accessories/FORT_Backpacks/Backpack_F_MED_RoyalDusk/Materials/MF_RoyalDusk_Crystal}
  MaterialExpressionReroute_1 = Reroute(None=@n2.Result)
  MaterialFunctionCall!(In=MaterialFunctionCall(Material=@MaterialExpressionReroute_1.Result, M=@n1.M, Pre Skinned Local Position=@n1.M, Distance Blend=@n1.M, Fresnel=@n1.M) {materialFunction=&MF_CharacterParent_2023_PostFX}.Result) {materialFunction=/Game/Effects/Niagara/MasterMaterials/Functions/MF_FX_MAEmissiveAdjustments}
  PinBase() {reroutePins [
    "In (MA)" {expression=&self:MaterialExpressionReroute_0}
    FX {expression=&self:MaterialExpressionReroute_2}
   ]}
  PinBase!(Result=@n2.Result) {reroutePins [Result {expression=&self:MaterialExpressionReroute_1}]}
 ]
}
