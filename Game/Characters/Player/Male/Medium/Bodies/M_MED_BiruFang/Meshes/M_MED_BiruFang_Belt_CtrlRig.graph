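# Control Rig for the Biru Fang medium male body that adjusts the belt bones as the thighs move, acting as a pose-driven corrective.
# A sequence runs two mirrored spherical pose readers, one per thigh, with debug drawing left on.
# Each reader's output weights local additive offsets that shift and rotate the shared dyn_belt_MSTR bone and rotate that side's rear belt master bone.
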ControlRigBlueprint /Game/Characters/Player/Male/Medium/Bodies/M_MED_BiruFang/Meshes/M_MED_BiruFang_Belt_CtrlRig
graphs [
 RigVMFunctionLibrary {links [], nodes []}
 RigVMModel {nodes [
   RigUnit_BeginExecution = &RigUnit_BeginExecution()
   SequenceAggregate = RigVMAggregateNode(ExecuteContext=@RigUnit_BeginExecution.ExecutePin)
   n1 = &RigUnit_SphericalPoseReader(
    ActiveRegionScaleFactors=(PositiveWidth=2.5,NegativeWidth=1,PositiveHeight=2.5,NegativeHeight=2.5)
    ActiveRegionSize=0.1
    Debug=(bDrawDebug=true,bDraw2D=false,bDrawLocalAxes=false,DebugScale=25,DebugSegments=20,DebugThickness=0.25)
    DriverAxis=(X=-1,Y=0,Z=0)
    DriverItem=(Type=Bone,Name="thigh_l")
    FalloffRegionScaleFactors=(PositiveWidth=1,NegativeWidth=0.8,PositiveHeight=1,NegativeHeight=1)
    FalloffSize=0.8
    FlipHeightScaling=true
    FlipWidthScaling=true
    OptionalParentItem=(Type=Bone,Name="None")
    RotationOffset=(X=90,Y=180,Z=0)
    ExecutePin=@SequenceAggregate.A
   )
   RigVMFunction_MathQuaternionFromEuler = &RigVMFunction_MathQuaternionFromEuler(Euler=(X=-20,Y=-10,Z=0))
   > &RigUnit_ModifyTransforms(
    ItemToModify=((Item=(Type=Bone,Name="dyn_belt_MSTR"),Transform=(Rotation=(X=0,Y=0,Z=0,W=1),Translation=(X=0.5,Y=0,Z=3.25),Scale3D=(X=1,Y=1,Z=1))))
    ItemToModify.0.Transform.Rotation=@RigVMFunction_MathQuaternionFromEuler.Result
    Weight=@n1.OutputParam
   )
   > &RigUnit_ModifyTransforms(
    ItemToModify=((Item=(Type=Bone,Name="dyn_l_b_belt_MSTR"),Transform=(Rotation=(X=0,Y=0,Z=0,W=1),Translation=(X=0,Y=0,Z=0),Scale3D=(X=1,Y=1,Z=1))))
    ItemToModify.0.Transform.Rotation=&RigVMFunction_MathQuaternionFromEuler(Euler=(X=0,Y=-20,Z=0)).Result
    Weight=@n1.OutputParam
   )
   n2 = &RigUnit_SphericalPoseReader(
    ActiveRegionScaleFactors=(PositiveWidth=2.5,NegativeWidth=1,PositiveHeight=2.5,NegativeHeight=2.5)
    ActiveRegionSize=0.1
    Debug=(bDrawDebug=true,bDraw2D=false,DebugScale=25,DebugSegments=20,DebugThickness=0.25)
    DriverAxis=(X=1,Y=0,Z=0)
    DriverItem=(Type=Bone,Name="thigh_r")
    FalloffRegionScaleFactors=(PositiveWidth=1,NegativeWidth=0.8,PositiveHeight=1,NegativeHeight=1)
    FalloffSize=0.8
    FlipHeightScaling=true
    FlipWidthScaling=true
    OptionalParentItem=(Type=Bone,Name="None")
    RotationOffset=(X=-90,Y=0,Z=0)
    ExecutePin=@SequenceAggregate.B
   )
   > &RigUnit_ModifyTransforms(
    ItemToModify=((Item=(Type=Bone,Name="dyn_belt_MSTR"),Transform=(Rotation=(X=0,Y=0,Z=0,W=1),Translation=(X=0.5,Y=0,Z=3.25),Scale3D=(X=1,Y=1,Z=1))))
    ItemToModify.0.Transform.Rotation=&RigVMFunction_MathQuaternionFromEuler(Euler=(X=20,Y=-10,Z=0)).Result
    Weight=@n2.OutputParam
   )
   > &RigUnit_ModifyTransforms(
    ItemToModify=((Item=(Type=Bone,Name="dyn_r_b_belt_MSTR"),Transform=(Rotation=(X=0,Y=0,Z=0,W=1),Translation=(X=0,Y=0,Z=0),Scale3D=(X=1,Y=1,Z=1))))
    ItemToModify.0.Transform.Rotation=&RigVMFunction_MathQuaternionFromEuler(Euler=(X=0,Y=20,Z=0)).Result
    Weight=@n2.OutputParam
   )
  ]}
 SequenceAggregate_1_ContainedGraph {nodes [
   n1 = RigVMFunctionEntryNode()
   SequenceAggregate = &RigVMFunction_Sequence(ExecuteContext=@n1.ExecuteContext)
   SequenceAggregate_1 = &RigVMFunction_Sequence(ExecuteContext=@SequenceAggregate.B)
   SequenceAggregate_2 = &RigVMFunction_Sequence(ExecuteContext=@SequenceAggregate_1.B)
   RigVMFunctionReturnNode(A=@SequenceAggregate.A, B=@SequenceAggregate_1.A, C=@SequenceAggregate_2.A, D=@SequenceAggregate_2.B)
  ]}
]
hierarchy [
 root {loc=<0 0 0>, rot=<0 0 0 1>}
 pelvis {parent=root, loc=<0 -0.7071 93.1191>, rot=<0.0127 0.707 -0.0127 -0.707>}
 spine_01 {parent=pelvis, loc=<6.7494 -0.6388 0>, rot=<0 0 0.034 -0.9994>}
 spine_02 {parent=spine_01, loc=<7.2518 -0.035 0>, rot=<0 0 0.0238 -0.9997>}
 spine_03 {parent=spine_02, loc=<6.0104 0 0>, rot=<0 0 0.0149 -0.9999>}
 spine_04 {parent=spine_03, loc=<8.1024 0 0>, rot=<0 0 -0.0132 -0.9999>}
 spine_05 {parent=spine_04, loc=<10.0146 0 0>, rot=<0 0 -0.0015 -1>}
 attach_cape {parent=spine_05, loc=<9.0172 -11.7141 0>, rot=<0.0616 -0.7044 -0.0616 -0.7044>}
 neck_01 {parent=spine_05, loc=<11.3138 0.1529 0>, rot=<0 0 -0.2003 -0.9797>}
 neck_02 {parent=neck_01, loc=<4.5627 0 0>, rot=<0 0 0.0504 -0.9987>}
 head {parent=neck_02, loc=<5.3914 0 0>, rot=<0 0 0.115 -0.9934>}
 attach_backpack {parent=spine_05, loc=<-0.1228 -13.4731 0>, rot=<-0.0093 -0.707 0.0093 -0.707>}
 clavicle_l {parent=spine_05, loc=<4.3575 -0.1466 -2.3223>, rot=<0.0465 -0.7289 -0.009 -0.6829>}
 upperarm_l {parent=clavicle_l, loc=<15.5111 0 0>, rot=<0.0226 -0.3234 -0.0088 -0.9459>}
 lowerarm_l {parent=upperarm_l, loc=<26.6514 0 0>, rot=<0.0391 -0.0099 -0.2453 -0.9686>}
 hand_l {parent=lowerarm_l, loc=<22.3491 0 0>, rot=<-0.7081 -0.0507 0.0975 -0.6975>}
 thumb_01_l {parent=hand_l, loc=<1.3118 -0.3191 -2.459>, rot=<0.6546 -0.4042 0.0664 -0.6354>}
 thumb_02_l {parent=thumb_01_l, loc=<5.9336 -0.0448 0.0401>, rot=<-0.0037 0.0213 0.298 -0.9543>}
 thumb_03_l {parent=thumb_02_l, loc=<2.2223 0.3682 -0.0907>, rot=<0.0118 -0.013 0.0066 -0.9998>}
 index_metacarpal_l {parent=hand_l, loc=<1.6535 0.0183 -2.96>, rot=<0.0158 -0.0124 -0.0173 -0.9996>}
 index_01_l {parent=index_metacarpal_l, loc=<8.3008 0.0357 -0.0524>, rot=<-0.0044 0.0485 0.0924 -0.9945>}
 index_02_l {parent=index_01_l, loc=<3.3587 -0.3004 0.0452>, rot=<-0.002 0.0027 0.3225 -0.9466>}
 index_03_l {parent=index_02_l, loc=<2.5578 -0.123 -0.0252>, rot=<-0.002 0.0067 0.2558 -0.9667>}
 middle_metacarpal_l {parent=hand_l, loc=<2.1856 -0.1725 -0.5984>, rot=<0.006 0.0168 -0.0329 -0.9993>}
 middle_01_l {parent=middle_metacarpal_l, loc=<7.3275 -0.2447 0.0129>, rot=<-0.0208 0.0042 0.0858 -0.9961>}
 middle_02_l {parent=middle_01_l, loc=<4.0123 0.1401 0.1175>, rot=<-0.0218 0.0258 0.3512 -0.9357>}
 middle_03_l {parent=middle_02_l, loc=<2.6077 0.0534 0.1373>, rot=<0.0082 0.0188 0.1708 -0.9851>}
 ring_metacarpal_l {parent=hand_l, loc=<2.2074 -0.3625 1.3638>, rot=<-0.0284 0.0722 -0.0458 -0.9959>}
 ring_01_l {parent=ring_metacarpal_l, loc=<7.0295 -0.6197 -0.1446>, rot=<-0.0489 -0.0252 0.1892 -0.9804>}
 ring_02_l {parent=ring_01_l, loc=<3.6174 0.4647 -0.1027>, rot=<0 0 0.2806 -0.9598>}
 ring_03_l {parent=ring_02_l, loc=<2.4152 0.114 0.0199>, rot=<0 0 0.1189 -0.9929>}
 pinky_metacarpal_l {parent=hand_l, loc=<1.7616 -0.9753 3.3814>, rot=<-0.0397 0.0814 -0.069 -0.9935>}
 pinky_01_l {parent=pinky_metacarpal_l, loc=<7.1709 -0.4831 -0.1094>, rot=<-0.0654 -0.0035 0.2242 -0.9723>}
 pinky_02_l {parent=pinky_01_l, loc=<2.8274 0.0893 -0.1043>, rot=<0.0037 0.0011 0.2854 -0.9584>}
 pinky_03_l {parent=pinky_02_l, loc=<2.2046 0.0307 -0.0078>, rot=<-0.0002 0.0098 0.1445 -0.9895>}
 lowerarm_twist_02_l {parent=lowerarm_l, loc=<9.6199 0 0>, rot=<0 0 0 -1>}
 lowerarm_twist_01_l {parent=lowerarm_l, loc=<17.8549 0 0>, rot=<0 0 0 -1>}
 dyn_l_cuff_buck_MSTR {parent=lowerarm_twist_01_l, loc=<1.7001 -1.2446 6.2189>, rot=<-0.0606 0.6391 0.209 -0.7377>}
 dyn_l_cuff_buck_01 {parent=dyn_l_cuff_buck_MSTR, loc=<0 0 0>, rot=<0 0 0 1>}
 upperarm_twist_01_l {parent=upperarm_l, loc=<0.0041 0 0>, rot=<0 0 0 1>}
 upperarm_twist_02_l {parent=upperarm_l, loc=<13.7941 0 0>, rot=<0 0 0 1>}
 dyn_shoulder_l {parent=clavicle_l, loc=<10.9006 1.3453 6.7569>, rot=<0.0014 0.0034 0.0377 -0.9993>}
 dyn_fluff_24_MSTR {parent=dyn_shoulder_l, loc=<-5.3599 6.9025 -1.2223>, rot=<0.897 0.1661 0.4067 0.0499>}
 dyn_fluff_24 {parent=dyn_fluff_24_MSTR, loc=<0 0 0>, rot=<0 0 0 1>}
 dyn_fluff_23_MSTR {parent=dyn_shoulder_l, loc=<-0.5239 6.8462 -2.5103>, rot=<0.8623 -0.0252 0.5027 0.0547>}
 dyn_fluff_23 {parent=dyn_fluff_23_MSTR, loc=<0 0 0>, rot=<0 0 0 1>}
 dyn_fluff_22_MSTR {parent=dyn_shoulder_l, loc=<0.7722 1.9593 0.269>, rot=<0.8906 -0.2958 0.0521 0.3416>}
 dyn_fluff_22 {parent=dyn_fluff_22_MSTR, loc=<0 0 0>, rot=<0 0 0 1>}
 dyn_fluff_21_MSTR {parent=dyn_shoulder_l, loc=<0.8465 -2.6135 2.3046>, rot=<0.7678 -0.2961 0.3305 0.4622>}
 dyn_fluff_21 {parent=dyn_fluff_21_MSTR, loc=<0 0 0>, rot=<0 0 0 1>}
 dyn_fluff_20_MSTR {parent=dyn_shoulder_l, loc=<-2.3512 -4.4467 2.088>, rot=<0.9678 -0.2492 -0.0123 0.0325>}
 dyn_fluff_20 {parent=dyn_fluff_20_MSTR, loc=<0 0 0>, rot=<0 0 0 1>}
 dyn_fluff_19_MSTR {parent=dyn_shoulder_l, loc=<-1.2019 -5.5033 6.229>, rot=<0.8377 -0.5431 -0.0425 0.0393>}
 dyn_fluff_19 {parent=dyn_fluff_19_MSTR, loc=<0 0 0>, rot=<0 0 0 1>}
 clavicle_r {parent=spine_05, loc=<4.3575 -0.1466 2.3223>, rot=<0.7289 0.0465 -0.6829 0.009>}
 upperarm_r {parent=clavicle_r, loc=<-15.5111 0 0>, rot=<0.0226 -0.3234 -0.0088 -0.9459>}
 lowerarm_r {parent=upperarm_r, loc=<-26.6514 0 0>, rot=<0.0391 -0.0099 -0.2453 -0.9686>}
 hand_r {parent=lowerarm_r, loc=<-22.3491 0 0>, rot=<-0.7081 -0.0507 0.0975 -0.6975>}
 thumb_01_r {parent=hand_r, loc=<-1.3118 0.3191 2.459>, rot=<0.6546 -0.4042 0.0664 -0.6354>}
 thumb_02_r {parent=thumb_01_r, loc=<-5.9336 0.0448 -0.0401>, rot=<-0.0037 0.0213 0.298 -0.9543>}
 thumb_03_r {parent=thumb_02_r, loc=<-2.2223 -0.3682 0.0907>, rot=<0.0118 -0.013 0.0066 -0.9998>}
 index_metacarpal_r {parent=hand_r, loc=<-1.6535 -0.0183 2.96>, rot=<0.0158 -0.0124 -0.0173 -0.9996>}
 index_01_r {parent=index_metacarpal_r, loc=<-8.3008 -0.0357 0.0524>, rot=<-0.0044 0.0485 0.0924 -0.9945>}
 index_02_r {parent=index_01_r, loc=<-3.3587 0.3004 -0.0452>, rot=<-0.002 0.0027 0.3225 -0.9466>}
 index_03_r {parent=index_02_r, loc=<-2.5578 0.123 0.0252>, rot=<-0.002 0.0067 0.2558 -0.9667>}
 middle_metacarpal_r {parent=hand_r, loc=<-2.1856 0.1725 0.5984>, rot=<0.006 0.0168 -0.0329 -0.9993>}
 middle_01_r {parent=middle_metacarpal_r, loc=<-7.3275 0.2447 -0.0129>, rot=<-0.0208 0.0042 0.0858 -0.9961>}
 middle_02_r {parent=middle_01_r, loc=<-4.0123 -0.1401 -0.1175>, rot=<-0.0218 0.0258 0.3512 -0.9357>}
 middle_03_r {parent=middle_02_r, loc=<-2.6077 -0.0534 -0.1373>, rot=<0.0082 0.0188 0.1708 -0.9851>}
 ring_metacarpal_r {parent=hand_r, loc=<-2.2074 0.3625 -1.3638>, rot=<-0.0284 0.0722 -0.0458 -0.9959>}
 ring_01_r {parent=ring_metacarpal_r, loc=<-7.0295 0.6197 0.1446>, rot=<-0.0489 -0.0252 0.1892 -0.9804>}
 ring_02_r {parent=ring_01_r, loc=<-3.6174 -0.4647 0.1027>, rot=<0 0 0.2806 -0.9598>}
 ring_03_r {parent=ring_02_r, loc=<-2.4152 -0.114 -0.0199>, rot=<0 0 0.1189 -0.9929>}
 pinky_metacarpal_r {parent=hand_r, loc=<-1.7616 0.9753 -3.3814>, rot=<-0.0397 0.0814 -0.069 -0.9935>}
 pinky_01_r {parent=pinky_metacarpal_r, loc=<-7.1709 0.4831 0.1094>, rot=<-0.0654 -0.0035 0.2242 -0.9723>}
 pinky_02_r {parent=pinky_01_r, loc=<-2.8274 -0.0893 0.1043>, rot=<0.0037 0.0011 0.2854 -0.9584>}
 pinky_03_r {parent=pinky_02_r, loc=<-2.2046 -0.0307 0.0078>, rot=<-0.0002 0.0098 0.1445 -0.9895>}
 lowerarm_twist_02_r {parent=lowerarm_r, loc=<-9.6199 0 0>, rot=<0 0 0 -1>}
 lowerarm_twist_01_r {parent=lowerarm_r, loc=<-17.8549 0 0>, rot=<0 0 0 -1>}
 dyn_r_cuff_buck_MSTR {parent=lowerarm_twist_01_r, loc=<-1.7001 1.2446 -6.2189>, rot=<-0.6391 -0.0606 0.7377 0.209>}
 dyn_r_cuff_buck_01 {parent=dyn_r_cuff_buck_MSTR, loc=<0 0 0>, rot=<0 0 0 1>}
 upperarm_twist_01_r {parent=upperarm_r, loc=<-0.0041 0 0>, rot=<0 0 0 -1>}
 upperarm_twist_02_r {parent=upperarm_r, loc=<-13.7941 0 0>, rot=<0 0 0 -1>}
 dyn_shoulder_r {parent=clavicle_r, loc=<-10.9006 -1.3453 -6.7569>, rot=<-0.0377 -0.9993 0.0014 -0.0034>}
 dyn_fluff_06_MSTR {parent=dyn_shoulder_r, loc=<0.8907 5.8814 3.3568>, rot=<0.8708 0.3327 0.0844 -0.3521>}
 dyn_fluff_06 {parent=dyn_fluff_06_MSTR, loc=<0 0 0>, rot=<0 0 0 1>}
 dyn_fluff_05_MSTR {parent=dyn_shoulder_r, loc=<3.4069 4.6357 1.6001>, rot=<0.7308 0.571 -0.3681 0.0667>}
 dyn_fluff_05 {parent=dyn_fluff_05_MSTR, loc=<0 0 0>, rot=<0 0 0 1>}
 dyn_fluff_04_MSTR {parent=dyn_shoulder_r, loc=<4.3393 1.4036 1.3646>, rot=<0.8165 0.4099 0.3304 -0.237>}
 dyn_fluff_04 {parent=dyn_fluff_04_MSTR, loc=<0 0 0>, rot=<0 0 0 1>}
 dyn_fluff_03_MSTR {parent=dyn_shoulder_r, loc=<0.8745 -3.9107 -1.8943>, rot=<0.9428 0.1352 0.277 -0.1268>}
 dyn_fluff_03 {parent=dyn_fluff_03_MSTR, loc=<0 0 0>, rot=<0 0 0 1>}
 dyn_fluff_02_MSTR {parent=dyn_shoulder_r, loc=<0.5747 -5.32 -3.7412>, rot=<0.949 -0.1617 0.2325 -0.1385>}
 dyn_fluff_02 {parent=dyn_fluff_02_MSTR, loc=<0 0 0>, rot=<0 0 0 1>}
 dyn_fluff_01_MSTR {parent=dyn_shoulder_r, loc=<-3.7404 -7.4014 -3.4541>, rot=<0.89 -0.1242 0.4316 -0.0794>}
 dyn_fluff_01 {parent=dyn_fluff_01_MSTR, loc=<0 0 0>, rot=<0 0 0 1>}
 dyn_fluff_18_MSTR {parent=spine_05, loc=<17.9521 -4.8793 -9.0473>, rot=<0.8635 -0.1509 -0.1535 0.456>}
 dyn_fluff_18 {parent=dyn_fluff_18_MSTR, loc=<0 0 0>, rot=<0 0 0 1>}
 dyn_fluff_17_MSTR {parent=spine_05, loc=<12.7179 -6.3012 -6.794>, rot=<-0.5592 0.3261 0.7301 -0.2189>}
 dyn_fluff_17 {parent=dyn_fluff_17_MSTR, loc=<0 0 0>, rot=<0 0 0 1>}
 dyn_fluff_16_MSTR {parent=spine_05, loc=<14.6373 -7.3725 -6.5335>, rot=<0.7473 -0.216 -0.4281 0.4599>}
 dyn_fluff_16 {parent=dyn_fluff_16_MSTR, loc=<0 0 0>, rot=<0 0 0 1>}
 dyn_fluff_15_MSTR {parent=spine_05, loc=<9.3456 -8.0572 -1.2037>, rot=<-0.5943 0.371 0.6292 -0.3366>}
 dyn_fluff_15 {parent=dyn_fluff_15_MSTR, loc=<0 0 0>, rot=<0 0 0 1>}
 dyn_fluff_14_MSTR {parent=spine_05, loc=<13.0998 -9.4016 -1.0711>, rot=<-0.4847 0.4604 0.5232 -0.5285>}
 dyn_fluff_14 {parent=dyn_fluff_14_MSTR, loc=<0 0 0>, rot=<0 0 0 1>}
 dyn_fluff_13_MSTR {parent=spine_05, loc=<19.2353 -5.8548 1.7821>, rot=<0.719 -0.2778 -0.4604 0.4404>}
 dyn_fluff_13 {parent=dyn_fluff_13_MSTR, loc=<0 0 0>, rot=<0 0 0 1>}
 dyn_fluff_12_MSTR {parent=spine_05, loc=<10.8629 -7.1366 0.3101>, rot=<-0.1973 0.6453 0.5424 -0.5004>}
 dyn_fluff_12 {parent=dyn_fluff_12_MSTR, loc=<0 0 0>, rot=<0 0 0 1>}
 dyn_fluff_11_MSTR {parent=spine_05, loc=<15.8881 -9.0889 4.8081>, rot=<0.6173 -0.3274 -0.601 0.3879>}
 dyn_fluff_11 {parent=dyn_fluff_11_MSTR, loc=<0 0 0>, rot=<0 0 0 1>}
 dyn_fluff_10_MSTR {parent=spine_05, loc=<19.0064 -6.7343 9.2467>, rot=<0.8508 -0.1818 -0.4449 0.2126>}
 dyn_fluff_10 {parent=dyn_fluff_10_MSTR, loc=<0 0 0>, rot=<0 0 0 1>}
 dyn_fluff_09_MSTR {parent=spine_05, loc=<12.854 -6.1054 9.0823>, rot=<0.5329 -0.6653 -0.4595 0.2497>}
 dyn_fluff_09 {parent=dyn_fluff_09_MSTR, loc=<0 0 0>, rot=<0 0 0 1>}
 dyn_fluff_08_MSTR {parent=spine_05, loc=<15.1469 -4.947 9.4025>, rot=<-0.4955 0.4952 0.1183 -0.7038>}
 dyn_fluff_08 {parent=dyn_fluff_08_MSTR, loc=<0 0 0>, rot=<0 0 0 1>}
 dyn_fluff_07_MSTR {parent=spine_05, loc=<11.3927 -5.4227 11.4656>, rot=<0.4811 -0.7185 -0.4397 0.2428>}
 dyn_fluff_07 {parent=dyn_fluff_07_MSTR, loc=<0 0 0>, rot=<0 0 0 1>}
 dyn_r_back_buck_MSTR {parent=spine_05, loc=<5.1391 -9.0317 9.8798>, rot=<0.1863 -0.9483 -0.0292 0.2551>}
 dyn_r_back_buck_01 {parent=dyn_r_back_buck_MSTR, loc=<0 0 0>, rot=<-0.0002 0.0262 -0.0698 -0.9972>}
 dyn_l_back_buck_MSTR {parent=spine_05, loc=<5.1391 -9.0317 -9.8798>, rot=<0.1863 -0.9483 0.0292 -0.2551>}
 dyn_l_back_buck_01 {parent=dyn_l_back_buck_MSTR, loc=<0 0 0>, rot=<0.0001 -0.0262 -0.0698 -0.9972>}
 dyn_r_chest_buck_MSTR {parent=spine_05, loc=<-0.4364 10.1439 10.4161>, rot=<-0.1317 -0.9531 -0.2227 0.1569>}
 dyn_r_chest_buck_01 {parent=dyn_r_chest_buck_MSTR, loc=<0 0 0>, rot=<0 0 0 1>}
 dyn_r_chest_buck_02 {parent=dyn_r_chest_buck_01, loc=<2.5034 0 0>, rot=<0.0104 0.0233 0.0503 -0.9984>}
 dyn_r_chest_buck_03 {parent=dyn_r_chest_buck_02, loc=<2.9243 0 0>, rot=<-0.0037 -0.0266 0.004 -0.9996>}
 dyn_l_chest_buck_MSTR {parent=spine_05, loc=<-0.4364 10.1439 -10.4161>, rot=<-0.1317 -0.9531 0.2227 -0.1569>}
 dyn_l_chest_buck_01 {parent=dyn_l_chest_buck_MSTR, loc=<0 0 0>, rot=<0 0 0 1>}
 dyn_l_chest_buck_02 {parent=dyn_l_chest_buck_01, loc=<2.5034 0 0>, rot=<-0.01 -0.0217 0.0501 -0.9985>}
 dyn_l_chest_buck_03 {parent=dyn_l_chest_buck_02, loc=<2.9331 0 0>, rot=<0.0033 0.0242 0.0045 -0.9997>}
 dyn_collar_l_01 {parent=spine_05, loc=<7.1945 10.081 -3.5363>, rot=<0.2296 0.0403 -0.1683 -0.9578>}
 dyn_collar_l_02 {parent=spine_05, loc=<8.7185 7.4481 -7.1328>, rot=<0.4055 0.0713 -0.1577 -0.8975>}
 dyn_collar_r_01 {parent=spine_05, loc=<7.1943 10.081 3.5363>, rot=<-0.2296 -0.0403 -0.1683 -0.9578>}
 dyn_collar_r_02 {parent=spine_05, loc=<8.7182 7.4481 7.1328>, rot=<-0.4055 -0.0713 -0.1577 -0.8975>}
 thigh_l {parent=pelvis, loc=<-2.5693 -0.0109 -7.7501>, rot=<0.0012 0.0572 0.0155 -0.9982>}
 calf_l {parent=thigh_l, loc=<-39.0484 0 0>, rot=<0.0006 -0.0161 -0.0391 -0.9991>}
 foot_l {parent=calf_l, loc=<-42.561 0.1077 0.4292>, rot=<0.0418 -0.0455 0.0343 -0.9975>}
 ball_l {parent=foot_l, loc=<-6.1599 13.8264 0.6806>, rot=<0.0001 -0.0001 -0.7063 -0.7079>}
 calf_twist_02_l {parent=calf_l, loc=<-18.8064 -0.0967 0.0397>, rot=<0.0237 0.0095 -0.0035 -0.9997>}
 dyn_l_u_calf_strap_MSTR {parent=calf_twist_02_l, loc=<9.8976 -3.7444 -5.4363>, rot=<0.0048 -0.0481 0.6536 -0.7553>}
 dyn_l_u_calf_strap_01 {parent=dyn_l_u_calf_strap_MSTR, loc=<0 0 0>, rot=<0 0 0 1>}
 dyn_l_u_calf_strap_02 {parent=dyn_l_u_calf_strap_01, loc=<3.4134 0 0>, rot=<-0.0237 0.0064 -0.0319 -0.9992>}
 dyn_l_u_calf_strap_03 {parent=dyn_l_u_calf_strap_02, loc=<1.8417 0 0>, rot=<0.0066 -0.0023 -0.0331 -0.9994>}
 dyn_l_u_calf_strap_04 {parent=dyn_l_u_calf_strap_03, loc=<2.3215 0 0>, rot=<0 0 0 1>}
 dyn_l_d_calf_strap_MSTR {parent=calf_twist_02_l, loc=<5.4348 -3.6078 -5.8389>, rot=<0.0649 -0.1327 0.6408 -0.7534>}
 dyn_l_d_calf_strap_01 {parent=dyn_l_d_calf_strap_MSTR, loc=<0 0 0>, rot=<0 0 0 1>}
 dyn_l_d_calf_strap_02 {parent=dyn_l_d_calf_strap_01, loc=<3.8912 0 0>, rot=<-0.0999 0.0277 -0.0259 -0.9943>}
 dyn_l_d_calf_strap_03 {parent=dyn_l_d_calf_strap_02, loc=<1.8414 0 0>, rot=<-0.0009 0.0003 -0.0287 -0.9996>}
 dyn_l_d_calf_strap_04 {parent=dyn_l_d_calf_strap_03, loc=<2.3202 0 0>, rot=<0 0 0 1>}
 calf_twist_01_l {parent=calf_l, loc=<-35.813 -0.0967 0.0397>, rot=<0.0237 0.0095 -0.0035 -0.9997>}
 dyn_l_ankle_buck_MSTR {parent=calf_twist_01_l, loc=<1.302 -0.7613 -7.1752>, rot=<-0.0462 -0.7424 0.0155 -0.6682>}
 dyn_l_ankle_buck_01 {parent=dyn_l_ankle_buck_MSTR, loc=<0 0 0>, rot=<0 0 0 1>}
 dyn_kneecap_l_MSTR {parent=calf_l, loc=<-0.7904 6.3006 -0.1374>, rot=<-0.6025 -0.7981 -0.0006 0.0017>}
 thigh_twist_01_l {parent=thigh_l, loc=<-21.6303 0 0>, rot=<0.0013 0 0 -1>}
 dyn_button_l_01_MSTR {parent=thigh_twist_01_l, loc=<9.7461 1.5993 -7.1578>, rot=<-0.0144 -0.7123 -0.0108 -0.7017>}
 dyn_button_l_01 {parent=dyn_button_l_01_MSTR, loc=<0 0 0>, rot=<0 0 0 1>}
 dyn_button_l_02_MSTR {parent=thigh_twist_01_l, loc=<4.0187 1.4442 -7.6865>, rot=<-0.0227 -0.6976 -0.0198 -0.7159>}
 dyn_button_l_02 {parent=dyn_button_l_02_MSTR, loc=<0 0 0>, rot=<0 0 0 1>}
 dyn_button_l_03_MSTR {parent=thigh_twist_01_l, loc=<-1.3209 1.2915 -8.0536>, rot=<-0.0481 -0.7163 -0.0433 -0.6948>}
 dyn_button_l_03 {parent=dyn_button_l_03_MSTR, loc=<0 0 0>, rot=<0 0 0 1>}
 thigh_r {parent=pelvis, loc=<-2.5693 -0.0109 7.7501>, rot=<0.0572 -0.0012 0.9982 0.0155>}
 calf_r {parent=thigh_r, loc=<39.0484 0 0>, rot=<0.0006 -0.0161 -0.0391 -0.9991>}
 foot_r {parent=calf_r, loc=<42.561 -0.1077 -0.4292>, rot=<0.0418 -0.0455 0.0343 -0.9975>}
 ball_r {parent=foot_r, loc=<6.1599 -13.8264 -0.6806>, rot=<0.0001 -0.0001 -0.7063 -0.7079>}
 calf_twist_02_r {parent=calf_r, loc=<18.8064 0.0967 -0.0397>, rot=<0.0237 0.0095 -0.0035 -0.9997>}
 dyn_r_u_calf_strap_MSTR {parent=calf_twist_02_r, loc=<-9.8976 3.7444 5.4363>, rot=<-0.0481 -0.0048 -0.7553 -0.6536>}
 dyn_r_u_calf_strap_01 {parent=dyn_r_u_calf_strap_MSTR, loc=<0 0 0>, rot=<0 0 0 1>}
 dyn_r_u_calf_strap_02 {parent=dyn_r_u_calf_strap_01, loc=<3.4134 0 0>, rot=<0.0237 -0.0064 -0.0319 -0.9992>}
 dyn_r_u_calf_strap_03 {parent=dyn_r_u_calf_strap_02, loc=<1.8417 0 0>, rot=<-0.0066 0.0023 -0.0331 -0.9994>}
 dyn_r_u_calf_strap_04 {parent=dyn_r_u_calf_strap_03, loc=<2.3215 0 0>, rot=<0 0 0 1>}
 dyn_r_d_calf_strap_MSTR {parent=calf_twist_02_r, loc=<-5.4348 3.6078 5.8389>, rot=<-0.1327 -0.0649 -0.7534 -0.6408>}
 dyn_r_d_calf_strap_01 {parent=dyn_r_d_calf_strap_MSTR, loc=<0 0 0>, rot=<0 0 0 1>}
 dyn_r_d_calf_strap_02 {parent=dyn_r_d_calf_strap_01, loc=<3.8912 0 0>, rot=<0.0999 -0.0277 -0.0259 -0.9943>}
 dyn_r_d_calf_strap_03 {parent=dyn_r_d_calf_strap_02, loc=<1.8414 0 0>, rot=<0.0009 -0.0003 -0.0287 -0.9996>}
 dyn_r_d_calf_strap_04 {parent=dyn_r_d_calf_strap_03, loc=<2.3202 0 0>, rot=<0 0 0 1>}
 calf_twist_01_r {parent=calf_r, loc=<35.813 0.0967 -0.0397>, rot=<0.0237 0.0095 -0.0035 -0.9997>}
 dyn_r_ankle_buck_MSTR {parent=calf_twist_01_r, loc=<-1.302 0.7613 7.1752>, rot=<0.7424 -0.0462 0.6682 0.0155>}
 dyn_r_ankle_buck_01 {parent=dyn_r_ankle_buck_MSTR, loc=<0 0 0>, rot=<0 0 0 1>}
 dyn_kneecap_r_MSTR {parent=calf_r, loc=<0.7904 -6.3006 0.1374>, rot=<0.7981 -0.6025 -0.0017 -0.0006>}
 thigh_twist_01_r {parent=thigh_r, loc=<21.6303 0 0>, rot=<0.0013 0 0 -1>}
 dyn_button_r_01_MSTR {parent=thigh_twist_01_r, loc=<-9.7461 -1.5993 7.1578>, rot=<0.7123 -0.0144 0.7017 -0.0108>}
 dyn_button_r_01 {parent=dyn_button_r_01_MSTR, loc=<0 0 0>, rot=<0 0 0 1>}
 dyn_button_r_02_MSTR {parent=thigh_twist_01_r, loc=<-4.0187 -1.4442 7.6865>, rot=<0.6976 -0.0227 0.7159 -0.0198>}
 dyn_button_r_02 {parent=dyn_button_r_02_MSTR, loc=<0 0 0>, rot=<0 0 0 1>}
 dyn_button_r_03_MSTR {parent=thigh_twist_01_r, loc=<1.3209 -1.2915 8.0536>, rot=<0.7163 -0.0481 0.6948 -0.0433>}
 dyn_button_r_03 {parent=dyn_button_r_03_MSTR, loc=<0 0 0>, rot=<0 0 0 1>}
 dyn_r_b_belt_MSTR {parent=pelvis, loc=<1.9256 -9.8751 7.3159>, rot=<0.1321 0.4949 -0.1917 -0.8372>}
 dyn_r_b_belt_01 {parent=dyn_r_b_belt_MSTR, loc=<0 0 0>, rot=<0 0 0 1>}
 dyn_r_b_belt_02 {parent=dyn_r_b_belt_01, loc=<8.747 0 0>, rot=<0.1014 0.0649 -0.4893 -0.8638>}
 dyn_l_b_belt_MSTR {parent=pelvis, loc=<1.9256 -9.8751 -7.3159>, rot=<-0.1321 -0.4949 -0.1917 -0.8372>}
 dyn_l_b_belt_01 {parent=dyn_l_b_belt_MSTR, loc=<0 0 0>, rot=<0 0 0 1>}
 dyn_l_b_belt_02 {parent=dyn_l_b_belt_01, loc=<8.747 0 0>, rot=<-0.1014 -0.0649 -0.4893 -0.8638>}
 dyn_belt_MSTR {parent=pelvis, loc=<2.7176 8.689 0>, rot=<-0.5089 -0.491 -0.491 -0.5089>}
 dyn_l_f_strap_MSTR {parent=dyn_belt_MSTR, loc=<-1.8985 -10.8814 -3.6399>, rot=<-0.2698 -0.328 0.6992 -0.5751>}
 dyn_l_f_strap_01 {parent=dyn_l_f_strap_MSTR, loc=<0 0 0>, rot=<0 0 0 1>}
 dyn_l_f_strap_02 {parent=dyn_l_f_strap_01, loc=<7.1094 0 0>, rot=<0.0062 -0.0142 0.216 -0.9763>}
 dyn_l_f_strap_03 {parent=dyn_l_f_strap_02, loc=<3.6326 0 0>, rot=<0.0631 -0.0458 0.3068 -0.9486>}
 dyn_l_f_belt_MSTR {parent=dyn_belt_MSTR, loc=<-1.3438 -9.4605 0.3239>, rot=<0.1362 0.25 0.8418 -0.4585>}
 dyn_l_f_belt_01 {parent=dyn_l_f_belt_MSTR, loc=<0 0 0>, rot=<0 0 0 1>}
 dyn_l_f_belt_02 {parent=dyn_l_f_belt_01, loc=<7.3057 0 0>, rot=<0.8536 -0.4953 -0.0732 0.1435>}
 dyn_r_f_belt_MSTR {parent=dyn_belt_MSTR, loc=<-1.3438 9.4605 0.3239>, rot=<-0.4585 -0.8418 -0.25 0.1362>}
 dyn_r_f_belt_01 {parent=dyn_r_f_belt_MSTR, loc=<0 0 0>, rot=<0 0 0 1>}
 dyn_r_f_belt_02 {parent=dyn_r_f_belt_01, loc=<7.3057 0 0>, rot=<0.8536 -0.4953 0.0732 -0.1435>}
 dyn_belt_01 {parent=dyn_belt_MSTR, loc=<0 0 0>, rot=<0 0 0 1>}
 dyn_l_b_strap_MSTR {parent=pelvis, loc=<1.0295 -9.6204 -6.9108>, rot=<-0.0421 -0.9404 -0.003 -0.3375>}
 dyn_l_b_strap_01 {parent=dyn_l_b_strap_MSTR, loc=<0 0 0>, rot=<0 0 0 1>}
 dyn_l_b_strap_02 {parent=dyn_l_b_strap_01, loc=<9.3066 0 0>, rot=<0.0151 0.0475 -0.2515 -0.9666>}
 dyn_l_b_strap_03 {parent=dyn_l_b_strap_02, loc=<5.4516 0 0>, rot=<0.0677 0.0603 -0.2755 -0.957>}
 dyn_l_b_strap_04 {parent=dyn_l_b_strap_03, loc=<5.1493 0 0>, rot=<0.9033 0.3624 -0.0151 -0.2291>}
]
