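# Physics asset for the face accessory mesh of the BiruFang medium male player character, covering the head and left and right dynamic collar bones.
# Capsule bodies sit on the head, spine_05 and two collar bones per side.
# Each collar bone is constrained directly to spine_05 rather than chained, with narrow angle limits that are widest on swing 1.
# The head-to-spine_05 constraint leaves both swings and twist free.
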PhysicsAsset /Game/Characters/Player/Male/Medium/Bodies/M_MED_BiruFang/Meshes/M_MED_BiruFang_FaceAcc_Physics
bodies {
 dyn_collar_l_01 [{shapeType=Capsule, length=5, radius=5, shapeName=dyn_collar_l_01_capsule, center=<1 4 0>, rotation=<rot 90 0 0>}]
 dyn_collar_l_02 [{shapeType=Capsule, length=5, radius=5, shapeName=dyn_collar_l_02_capsule, center=<1 4 0>, rotation=<rot 90 0 0>}]
 dyn_collar_r_01 [{shapeType=Capsule, length=5, radius=5, shapeName=dyn_collar_r_01_capsule, center=<1 4 0>, rotation=<rot 90 0 0>}]
 dyn_collar_r_02 [{shapeType=Capsule, length=5, radius=5, shapeName=dyn_collar_r_02_capsule, center=<1 4 0>, rotation=<rot 90 0 0>}]
 head [{shapeType=Capsule, length=3, radius=10, shapeName=head_capsule, center=<0 2.258 0>, rotation=<rot 90 0 0>}]
 spine_05 [{shapeType=Capsule, length=0.505, radius=0.505, shapeName=spine_05_capsule}]
}
constraints [
 {
  bone1Name=head
  bone2Name=spine_05
  swing1LimitDegrees=45
  swing1Motion=Free
  swing2LimitDegrees=45
  swing2Motion=Free
  twistLimitDegrees=45
  twistMotion=Free
 }
 {bone1Name=dyn_collar_l_01, bone2Name=spine_05, swing1LimitDegrees=20, swing2LimitDegrees=1, twistLimitDegrees=1}
 {bone1Name=dyn_collar_l_02, bone2Name=spine_05, swing1LimitDegrees=20, swing2LimitDegrees=1, twistLimitDegrees=1}
 {bone1Name=dyn_collar_r_01, bone2Name=spine_05, swing1LimitDegrees=20, swing2LimitDegrees=1, twistLimitDegrees=1}
 {bone1Name=dyn_collar_r_02, bone2Name=spine_05, swing1LimitDegrees=20, swing2LimitDegrees=1, twistLimitDegrees=1}
]
