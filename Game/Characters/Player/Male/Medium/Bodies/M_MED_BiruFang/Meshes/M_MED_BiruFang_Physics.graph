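# Physics asset for the medium male BiruFang player body, covering dynamic belt and strap bones around the hips and left thigh.
# Capsule bodies form four belt chains (left/right, front/back) rooted on the pelvis and two strap chains (front/back) rooted on thigh_l.
# Each chain's tiny MSTR root and the thigh-to-pelvis joint rotate freely, while links below the roots have tight swing and twist limits.
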PhysicsAsset /Game/Characters/Player/Male/Medium/Bodies/M_MED_BiruFang/Meshes/M_MED_BiruFang_Physics
bodies {
 dyn_l_b_belt_01 [{shapeType=Capsule, length=3, radius=5, shapeName=dyn_l_b_belt_01_capsule, center=<0 -4 0>, rotation=<rot 90 0 0>}]
 dyn_l_b_belt_02 [{shapeType=Capsule, length=3, radius=5, shapeName=dyn_l_b_belt_02_capsule, center=<0 -4 0>, rotation=<rot 90 0 0>}]
 dyn_l_b_belt_MSTR [{shapeType=Capsule, length=0.505, radius=0.505, shapeName=dyn_l_b_belt_MSTR_capsule, center=<0 -0.2455 0>}]
 dyn_l_b_strap_01 [{shapeType=Capsule, length=4, radius=4, shapeName=dyn_l_b_strap_01_capsule, center=<0 -1 2.5>, rotation=<rot 90 0 0>}]
 dyn_l_b_strap_02 [{shapeType=Capsule, length=4, radius=4, shapeName=dyn_l_b_strap_02_capsule, center=<0 -1 2.5>, rotation=<rot 90 0 0>}]
 dyn_l_b_strap_03 [{shapeType=Capsule, length=4, radius=4, shapeName=dyn_l_b_strap_03_capsule, center=<0 -1 2.5>, rotation=<rot 90 0 0>}]
 dyn_l_b_strap_04 [{shapeType=Capsule, length=12, radius=5, shapeName=dyn_l_b_strap_04_capsule, center=<6 1 -2.5>, rotation=<rot 120 0 0>}]
 dyn_l_b_strap_MSTR [{
   shapeType=Capsule
   length=0.505
   radius=0.505
   shapeName=dyn_l_b_strap_MSTR_capsule
   center=<-0.1598 0.0983 0.4019>
   rotation=<rot 5.7997 -90 -12.8391>
  }]
 dyn_l_f_belt_01 [{shapeType=Capsule, length=3, radius=5, shapeName=dyn_l_f_belt_01_capsule, center=<0 4 0>, rotation=<rot 90 0 0>}]
 dyn_l_f_belt_02 [{shapeType=Capsule, length=3, radius=5, shapeName=dyn_l_f_belt_02_capsule, center=<0 -4 0>, rotation=<rot 90 0 0>}]
 dyn_l_f_belt_MSTR [{shapeType=Capsule, length=0.505, radius=0.505, shapeName=dyn_l_f_belt_MSTR_capsule, rotation=<rot 0 -90 0>}]
 dyn_l_f_strap_01 [{shapeType=Capsule, length=4, radius=5, shapeName=dyn_l_f_strap_01_capsule, center=<0 1 2.5>, rotation=<rot 90 0 0>}]
 dyn_l_f_strap_02 [{shapeType=Capsule, length=8, radius=5, shapeName=dyn_l_f_strap_02_capsule, center=<3 1 2.5>, rotation=<rot 90 0 0>}]
 dyn_l_f_strap_03 [{shapeType=Capsule, length=4, radius=4, shapeName=dyn_l_f_strap_03_capsule, center=<0 1 2.5>, rotation=<rot 90 0 0>}]
 dyn_l_f_strap_MSTR [{
   shapeType=Capsule
   length=0.505
   radius=0.505
   shapeName=dyn_l_f_strap_MSTR_capsule
   center=<-0.1663 -0.0092 0.0302>
   rotation=<rot -16.1861 -90 0.9434>
  }]
 dyn_r_b_belt_01 [{shapeType=Capsule, length=3, radius=5, shapeName=dyn_r_b_belt_01_capsule, center=<0 -4 0>, rotation=<rot 90 0 0>}]
 dyn_r_b_belt_02 [{shapeType=Capsule, length=3, radius=5, shapeName=dyn_r_b_belt_02_capsule, center=<0 -4 0>, rotation=<rot 90 0 0>}]
 dyn_r_b_belt_MSTR [{shapeType=Capsule, length=0.505, radius=0.505, shapeName=dyn_r_b_belt_MSTR_capsule, center=<0 -0.0333 0>}]
 dyn_r_f_belt_01 [{shapeType=Capsule, length=3, radius=5, shapeName=dyn_r_f_belt_01_capsule, center=<0 4 0>, rotation=<rot 90 0 0>}]
 dyn_r_f_belt_02 [{shapeType=Capsule, length=3, radius=5, shapeName=dyn_r_f_belt_02_capsule, center=<0 -4 0>, rotation=<rot 90 0 0>}]
 dyn_r_f_belt_MSTR [{
   shapeType=Capsule
   length=0.505
   radius=0.505
   shapeName=dyn_r_f_belt_MSTR_capsule
   center=<-0.043 0.2054 -0.837>
   rotation=<rot -0.7952 -90 -0.0898>
  }]
 pelvis [{shapeType=Capsule, length=12, radius=8, shapeName=pelvis_capsule, rotation=<rot 0 90 0>}]
 thigh_l [{shapeType=Capsule, length=24, radius=9, shapeName=thigh_l_capsule, center=<-16 0 0>, rotation=<rot 90 0 0>}]
}
constraints [
 {
  bone1Name=dyn_l_b_belt_MSTR
  bone2Name=pelvis
  swing1LimitDegrees=45
  swing1Motion=Free
  swing2LimitDegrees=45
  swing2Motion=Free
  twistLimitDegrees=45
  twistMotion=Free
 }
 {bone1Name=dyn_l_b_belt_01, bone2Name=dyn_l_b_belt_MSTR, swing1LimitDegrees=1, swing2LimitDegrees=10, twistLimitDegrees=10}
 {bone1Name=dyn_l_b_belt_02, bone2Name=dyn_l_b_belt_01, swing1LimitDegrees=1, swing2LimitDegrees=10, twistLimitDegrees=10}
 {
  bone1Name=dyn_l_f_belt_MSTR
  bone2Name=pelvis
  swing1LimitDegrees=45
  swing1Motion=Free
  swing2LimitDegrees=45
  swing2Motion=Free
  twistLimitDegrees=45
  twistMotion=Free
 }
 {bone1Name=dyn_l_f_belt_01, bone2Name=dyn_l_f_belt_MSTR, swing1LimitDegrees=1, swing2LimitDegrees=10, twistLimitDegrees=10}
 {bone1Name=dyn_l_f_belt_02, bone2Name=dyn_l_f_belt_01, swing1LimitDegrees=1, swing2LimitDegrees=10, twistLimitDegrees=10}
 {
  bone1Name=dyn_r_b_belt_MSTR
  bone2Name=pelvis
  swing1LimitDegrees=45
  swing1Motion=Free
  swing2LimitDegrees=45
  swing2Motion=Free
  twistLimitDegrees=45
  twistMotion=Free
 }
 {bone1Name=dyn_r_b_belt_01, bone2Name=dyn_r_b_belt_MSTR, swing1LimitDegrees=1, swing2LimitDegrees=10, twistLimitDegrees=10}
 {bone1Name=dyn_r_b_belt_02, bone2Name=dyn_r_b_belt_01, swing1LimitDegrees=1, swing2LimitDegrees=10, twistLimitDegrees=10}
 {
  bone1Name=dyn_r_f_belt_MSTR
  bone2Name=pelvis
  swing1LimitDegrees=45
  swing1Motion=Free
  swing2LimitDegrees=45
  swing2Motion=Free
  twistLimitDegrees=45
  twistMotion=Free
 }
 {bone1Name=dyn_r_f_belt_01, bone2Name=dyn_r_f_belt_MSTR, swing1LimitDegrees=1, swing2LimitDegrees=10, twistLimitDegrees=10}
 {bone1Name=dyn_r_f_belt_02, bone2Name=dyn_r_f_belt_01, swing1LimitDegrees=1, swing2LimitDegrees=10, twistLimitDegrees=10}
 {
  bone1Name=thigh_l
  bone2Name=pelvis
  swing1LimitDegrees=45
  swing1Motion=Free
  swing2LimitDegrees=45
  swing2Motion=Free
  twistLimitDegrees=45
  twistMotion=Free
 }
 {
  bone1Name=dyn_l_b_strap_MSTR
  bone2Name=thigh_l
  swing1LimitDegrees=45
  swing1Motion=Free
  swing2LimitDegrees=45
  swing2Motion=Free
  twistLimitDegrees=45
  twistMotion=Free
 }
 {bone1Name=dyn_l_b_strap_01, bone2Name=dyn_l_b_strap_MSTR, swing1LimitDegrees=3, swing2LimitDegrees=8, twistLimitDegrees=1}
 {bone1Name=dyn_l_b_strap_02, bone2Name=dyn_l_b_strap_01, swing1LimitDegrees=3, swing2LimitDegrees=8, twistLimitDegrees=1}
 {bone1Name=dyn_l_b_strap_03, bone2Name=dyn_l_b_strap_02, swing1LimitDegrees=3, swing2LimitDegrees=8, twistLimitDegrees=1}
 {bone1Name=dyn_l_b_strap_04, bone2Name=dyn_l_b_strap_03, swing1LimitDegrees=3, swing2LimitDegrees=8, twistLimitDegrees=1}
 {
  bone1Name=dyn_l_f_strap_MSTR
  bone2Name=thigh_l
  swing1LimitDegrees=45
  swing1Motion=Free
  swing2LimitDegrees=45
  swing2Motion=Free
  twistLimitDegrees=45
  twistMotion=Free
 }
 {bone1Name=dyn_l_f_strap_01, bone2Name=dyn_l_f_strap_MSTR, swing1LimitDegrees=1, swing2LimitDegrees=10, twistLimitDegrees=1}
 {bone1Name=dyn_l_f_strap_02, bone2Name=dyn_l_f_strap_01, swing1LimitDegrees=1, swing2LimitDegrees=10, twistLimitDegrees=1}
 {bone1Name=dyn_l_f_strap_03, bone2Name=dyn_l_f_strap_02, swing1LimitDegrees=1, swing2LimitDegrees=10, twistLimitDegrees=1}
]
