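# Secondary-motion AnimBlueprint for the BiruFang medium male player body, copying its pose from the attached parent mesh.
# Fur, calf and belt Control Rigs adjust that pose, with calf and belt alphas driven by crouch, jump and fall state.
# Bone constraints and additive offsets then correct kneecaps and a belt strap.
# Chained AnimDynamics nodes add spring motion to buttons, buckles, straps and fluff, the fluff taking alpha and stiffness from variables.
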AnimBlueprint /Game/Characters/Player/Male/Medium/Bodies/M_MED_BiruFang/Meshes/M_MED_BiruFang_AnimBP
target_skeleton=&/M_MED_BiruFang_Skeleton
graphs [
 AnimGraph {nodes [
   n1 = Control Rig(Source=Copy Pose From Mesh() {bUseAttachedParent=true}) {
    alphaBoolBlend {blendOption=Linear}
    controlRigAssetReference {blueprintRigClass=&/M_MED_BiruFang_Fur_CtrlRig.M_MED_BiruFang_Fur_CtrlRig_C}
   }
   n2 = Control Rig(Source=@n1, Alpha=Get isCoruchingJumpingFalling() -> (isCoruchingJumpingFalling:double)) {
    alphaBoolBlend {blendOption=Linear}
    alphaScaleBiasClamp {bInterpResult=true}
    controlRigAssetReference {blueprintRigClass=&/M_MED_BiruFang_Calf_CtrlRig.M_MED_BiruFang_Calf_CtrlRig_C}
   }
   n3 = Get isCrouching() -> (isCrouching:double)
   n4 = Control Rig(Source=@n2, Alpha=@n3) {
    alphaBoolBlend {blendOption=Linear}
    controlRigAssetReference {blueprintRigClass=&/M_MED_BiruFang_Belt_CtrlRig.M_MED_BiruFang_Belt_CtrlRig_C}
   }
   n5 = "Constraint - dyn_kneecap_l_MSTR (P)"(ComponentPose=Local To Component(LocalPose=@n4), Alpha=0.35, ConstraintWeights_0=1) {
    alphaBoolBlend {blendOption=Linear}
    boneToModify {boneName=dyn_kneecap_l_MSTR}
    constraintSetup [{targetBone {boneName=thigh_l}}]
    constraintWeights [1]
   }
   n6 = "Constraint - dyn_kneecap_r_MSTR (P)"(ComponentPose=@n5, Alpha=0.35, ConstraintWeights_0=1) {
    alphaBoolBlend {blendOption=Linear}
    boneToModify {boneName=dyn_kneecap_r_MSTR}
    constraintSetup [{targetBone {boneName=thigh_r}}]
    constraintWeights [1]
   }
   n7 = Get isNotCrouching() -> (isNotCrouching:double)
   n8 = "Constraint - dyn_l_f_strap_MSTR (P)"(ComponentPose=@n6, Alpha=@n7, ConstraintWeights_0=1) {
    alphaBoolBlend {blendOption=Linear}
    boneToModify {boneName=dyn_l_f_strap_MSTR}
    constraintSetup [{targetBone {boneName=dyn_belt_MSTR}}]
    constraintWeights [1]
   }
   n9 = "Constraint - dyn_l_f_strap_MSTR (PP)"!(ComponentPose=@n8, Alpha=@n3, ConstraintWeights_0=0.4, ConstraintWeights_1:float=0.6) -> (Pose) {
    alphaBoolBlend {blendOption=Linear}
    boneToModify {boneName=dyn_l_f_strap_MSTR}
    constraintSetup [{targetBone {boneName=dyn_belt_MSTR}}, {targetBone {boneName=thigh_l}}]
    constraintWeights [1,1]
   }
   n10 = "Transform (Modify) Bone - Bone: dyn_l_f_strap_MSTR"(ComponentPose=@n9, Alpha=@n7, Translation="-0.262,0.000,1.154", Rotation="0,0,0", Scale="1,1,1") {
    rotationMode=BMM_Additive
    translationMode=BMM_Additive
    alphaBoolBlend {blendOption=Linear}
    boneToModify {boneName=dyn_l_f_strap_MSTR}
    translation {x=-0.262, z=1.154}
   }
   n11 = "Transform (Modify) Bone - Bone: dyn_l_f_strap_MSTR"(ComponentPose=@n10, Alpha=@n3, Translation="-2.195,2.318,1.986", Rotation="-13.329,-24.904,-13.560", Scale="1,1,1") {
    rotationMode=BMM_Additive
    translationMode=BMM_Additive
    alphaBoolBlend {blendOption=Linear}
    boneToModify {boneName=dyn_l_f_strap_MSTR}
    rotation=<rot -13.329 -24.904 -13.56>
    translation=<-2.195 2.318 1.986>
   }
   n12 = Anim Dynamics - Bone: dyn_button_l_01(ComponentPose=@n11, Alpha=1, ExternalForce="0,0,0") {
    angularSpringConstant=500
    bAngularSpring=true
    bUseGravityOverride=true
    simulationSpace=RootRelative
    alphaBoolBlend {blendOption=Linear}
    boundBone {boneName=dyn_button_l_01}
    componentAppliedLinearAccClamp=<1000 1000 1000>
    componentLinearAccScale=<0.1 0.1 0.1>
    componentLinearVelScale=<0.1 0.1 0.1>
    gravityOverride {z=-980}
    physicsBodyDefinitions [{
      boundBone {boneName=dyn_button_l_01}
      boxExtents=<6 12 12>
      constraintSetup {angularLimitsMax=<3 3 8>, angularLimitsMin=<-3 -3 -8>, angularTarget=<10 0 0>}
      localJointOffset=<2 -2 0>
     }]
    relativeSpaceBone {boneName=calf_r}
   }
   n13 = Anim Dynamics - Bone: dyn_button_l_02(ComponentPose=@n12, Alpha=1, ExternalForce="0,0,0") {
    angularSpringConstant=500
    bAngularSpring=true
    bUseGravityOverride=true
    simulationSpace=RootRelative
    alphaBoolBlend {blendOption=Linear}
    boundBone {boneName=dyn_button_l_02}
    componentAppliedLinearAccClamp=<1000 1000 1000>
    componentLinearAccScale=<0.1 0.1 0.1>
    componentLinearVelScale=<0.1 0.1 0.1>
    gravityOverride {z=-980}
    physicsBodyDefinitions [{
      boundBone {boneName=dyn_button_l_02}
      boxExtents=<6 12 12>
      constraintSetup {angularLimitsMax=<3 3 8>, angularLimitsMin=<-3 -3 -8>, angularTarget=<10 0 0>}
      localJointOffset=<2 -2 0>
     }]
    relativeSpaceBone {boneName=calf_r}
   }
   n14 = Anim Dynamics - Bone: dyn_button_l_03(ComponentPose=@n13, Alpha=1, ExternalForce="0,0,0") {
    angularSpringConstant=500
    bAngularSpring=true
    bUseGravityOverride=true
    simulationSpace=RootRelative
    alphaBoolBlend {blendOption=Linear}
    boundBone {boneName=dyn_button_l_03}
    componentAppliedLinearAccClamp=<1000 1000 1000>
    componentLinearAccScale=<0.1 0.1 0.1>
    componentLinearVelScale=<0.1 0.1 0.1>
    gravityOverride {z=-980}
    physicsBodyDefinitions [{
      boundBone {boneName=dyn_button_l_03}
      boxExtents=<6 12 12>
      constraintSetup {angularLimitsMax=<3 3 8>, angularLimitsMin=<-3 -3 -8>, angularTarget=<10 0 0>}
      localJointOffset=<2 -2 0>
     }]
    relativeSpaceBone {boneName=calf_r}
   }
   n15 = Anim Dynamics - Bone: dyn_button_r_01(ComponentPose=@n14, Alpha=1, ExternalForce="0,0,0") {
    angularSpringConstant=500
    bAngularSpring=true
    bUseGravityOverride=true
    simulationSpace=RootRelative
    alphaBoolBlend {blendOption=Linear}
    boundBone {boneName=dyn_button_r_01}
    componentAppliedLinearAccClamp=<1000 1000 1000>
    componentLinearAccScale=<0.1 0.1 0.1>
    componentLinearVelScale=<0.1 0.1 0.1>
    gravityOverride {z=-980}
    physicsBodyDefinitions [{
      boundBone {boneName=dyn_button_r_01}
      boxExtents=<6 12 12>
      constraintSetup {angularLimitsMax=<3 3 8>, angularLimitsMin=<-3 -3 -8>, angularTarget=<10 0 0>}
      localJointOffset=<2 -2 0>
     }]
    relativeSpaceBone {boneName=calf_r}
   }
   n16 = Anim Dynamics - Bone: dyn_button_r_02(ComponentPose=@n15, Alpha=1, ExternalForce="0,0,0") {
    angularSpringConstant=500
    bAngularSpring=true
    bUseGravityOverride=true
    simulationSpace=RootRelative
    alphaBoolBlend {blendOption=Linear}
    boundBone {boneName=dyn_button_r_02}
    componentAppliedLinearAccClamp=<1000 1000 1000>
    componentLinearAccScale=<0.1 0.1 0.1>
    componentLinearVelScale=<0.1 0.1 0.1>
    gravityOverride {z=-980}
    physicsBodyDefinitions [{
      boundBone {boneName=dyn_button_r_02}
      boxExtents=<6 12 12>
      constraintSetup {angularLimitsMax=<3 3 8>, angularLimitsMin=<-3 -3 -8>, angularTarget=<10 0 0>}
      localJointOffset=<2 -2 0>
     }]
    relativeSpaceBone {boneName=calf_r}
   }
   n17 = Anim Dynamics - Bone: dyn_button_r_03(ComponentPose=@n16, Alpha=1, ExternalForce="0,0,0") {
    angularSpringConstant=500
    bAngularSpring=true
    bUseGravityOverride=true
    simulationSpace=RootRelative
    alphaBoolBlend {blendOption=Linear}
    boundBone {boneName=dyn_button_r_03}
    componentAppliedLinearAccClamp=<1000 1000 1000>
    componentLinearAccScale=<0.1 0.1 0.1>
    componentLinearVelScale=<0.1 0.1 0.1>
    gravityOverride {z=-980}
    physicsBodyDefinitions [{
      boundBone {boneName=dyn_button_r_03}
      boxExtents=<6 12 12>
      constraintSetup {angularLimitsMax=<3 3 8>, angularLimitsMin=<-3 -3 -8>, angularTarget=<10 0 0>}
      localJointOffset=<2 -2 0>
     }]
    relativeSpaceBone {boneName=calf_r}
   }
   n18 = Anim Dynamics - Bone: dyn_r_cuff_buck_01(ComponentPose=@n17, Alpha=1, ExternalForce="0,0,0") {
    angularSpringConstant=500
    bAngularSpring=true
    bUseGravityOverride=true
    simulationSpace=RootRelative
    alphaBoolBlend {blendOption=Linear}
    boundBone {boneName=dyn_r_cuff_buck_01}
    componentAppliedLinearAccClamp=<1000 1000 1000>
    componentLinearAccScale=<0.1 0.1 0.1>
    componentLinearVelScale=<0.1 0.1 0.1>
    gravityOverride {z=-980}
    physicsBodyDefinitions [{
      boundBone {boneName=dyn_r_cuff_buck_01}
      boxExtents=<16 20 20>
      constraintSetup {angularLimitsMax=<1 1 20>, angularLimitsMin=<-1 -1 -20>, angularTarget=<10 0 0>}
      localJointOffset=<3 0 0>
     }]
   }
   n19 = Anim Dynamics - Bone: dyn_l_cuff_buck_01(ComponentPose=@n18, Alpha=1, ExternalForce="0,0,0") {
    angularSpringConstant=500
    bAngularSpring=true
    bUseGravityOverride=true
    simulationSpace=RootRelative
    alphaBoolBlend {blendOption=Linear}
    boundBone {boneName=dyn_l_cuff_buck_01}
    componentAppliedLinearAccClamp=<1000 1000 1000>
    componentLinearAccScale=<0.1 0.1 0.1>
    componentLinearVelScale=<0.1 0.1 0.1>
    gravityOverride {z=-980}
    physicsBodyDefinitions [{
      boundBone {boneName=dyn_l_cuff_buck_01}
      boxExtents=<16 20 20>
      constraintSetup {angularLimitsMax=<1 1 20>, angularLimitsMin=<-1 -1 -20>, angularTarget=<10 0 0>}
      localJointOffset=<3 0 0>
     }]
   }
   n20 = Anim Dynamics - Bone: dyn_r_ankle_buck_01(ComponentPose=@n19, Alpha=1, ExternalForce="0,0,0") {
    angularSpringConstant=500
    bAngularSpring=true
    bUseGravityOverride=true
    simulationSpace=RootRelative
    alphaBoolBlend {blendOption=Linear}
    boundBone {boneName=dyn_r_ankle_buck_01}
    componentAppliedLinearAccClamp=<1000 1000 1000>
    componentLinearAccScale=<0.1 0.1 0.1>
    componentLinearVelScale=<0.1 0.1 0.1>
    gravityOverride {z=-980}
    physicsBodyDefinitions [{
      boundBone {boneName=dyn_r_ankle_buck_01}
      boxExtents=<16 20 20>
      constraintSetup {angularLimitsMax=<1 1 20>, angularLimitsMin=<-1 -1 -20>, angularTarget=<10 0 0>}
      localJointOffset=<3 0 0>
     }]
   }
   n21 = Anim Dynamics - Bone: dyn_l_ankle_buck_01(ComponentPose=@n20, Alpha=1, ExternalForce="0,0,0") {
    angularSpringConstant=500
    bAngularSpring=true
    bUseGravityOverride=true
    simulationSpace=RootRelative
    alphaBoolBlend {blendOption=Linear}
    boundBone {boneName=dyn_l_ankle_buck_01}
    componentAppliedLinearAccClamp=<1000 1000 1000>
    componentLinearAccScale=<0.1 0.1 0.1>
    componentLinearVelScale=<0.1 0.1 0.1>
    gravityOverride {z=-980}
    physicsBodyDefinitions [{
      boundBone {boneName=dyn_l_ankle_buck_01}
      boxExtents=<16 20 20>
      constraintSetup {angularLimitsMax=<1 1 20>, angularLimitsMin=<-1 -1 -20>, angularTarget=<10 0 0>}
      localJointOffset=<3 0 0>
     }]
   }
   n22 = Anim Dynamics - Bone: dyn_r_u_calf_strap_01(ComponentPose=@n21, Alpha=1, ExternalForce="0,0,0") {
    angularSpringConstant=1800
    bAngularSpring=true
    bUseGravityOverride=true
    simulationSpace=RootRelative
    alphaBoolBlend {blendOption=Linear}
    boundBone {boneName=dyn_r_u_calf_strap_01}
    componentAppliedLinearAccClamp=<1000 1000 1000>
    componentLinearAccScale=<0.1 0.1 0.1>
    componentLinearVelScale=<0.1 0.1 0.1>
    gravityOverride {z=-980}
    physicsBodyDefinitions [{
      boundBone {boneName=dyn_r_u_calf_strap_01}
      boxExtents=<10 20 6>
      constraintSetup {angularLimitsMax=<2 5 3>, angularLimitsMin=<-2 -5 -3>, angularTarget=<10 0 0>}
      localJointOffset=<2 -2 0>
     }]
    relativeSpaceBone {boneName=calf_r}
   }
   n23 = Anim Dynamics - Bone: dyn_r_u_calf_strap_02(ComponentPose=@n22, Alpha=1, ExternalForce="0,0,0") {
    angularSpringConstant=1800
    bAngularSpring=true
    bUseGravityOverride=true
    simulationSpace=RootRelative
    alphaBoolBlend {blendOption=Linear}
    boundBone {boneName=dyn_r_u_calf_strap_02}
    componentAppliedLinearAccClamp=<1000 1000 1000>
    componentLinearAccScale=<0.1 0.1 0.1>
    componentLinearVelScale=<0.1 0.1 0.1>
    gravityOverride {z=-980}
    physicsBodyDefinitions [{
      boundBone {boneName=dyn_r_u_calf_strap_02}
      boxExtents=<10 20 6>
      constraintSetup {angularLimitsMax=<2 5 3>, angularLimitsMin=<-2 -5 -3>, angularTarget=<10 0 0>}
      localJointOffset=<2 -2 0>
     }]
    relativeSpaceBone {boneName=calf_r}
   }
   n24 = Anim Dynamics - Bone: dyn_r_u_calf_strap_03(ComponentPose=@n23, Alpha=1, ExternalForce="0,0,0") {
    angularSpringConstant=1800
    bAngularSpring=true
    bUseGravityOverride=true
    simulationSpace=RootRelative
    alphaBoolBlend {blendOption=Linear}
    boundBone {boneName=dyn_r_u_calf_strap_03}
    componentAppliedLinearAccClamp=<1000 1000 1000>
    componentLinearAccScale=<0.1 0.1 0.1>
    componentLinearVelScale=<0.1 0.1 0.1>
    gravityOverride {z=-980}
    physicsBodyDefinitions [{
      boundBone {boneName=dyn_r_u_calf_strap_03}
      boxExtents=<10 20 6>
      constraintSetup {angularLimitsMax=<2 5 3>, angularLimitsMin=<-2 -5 -3>, angularTarget=<10 0 0>}
      localJointOffset=<2 -2 0>
     }]
    relativeSpaceBone {boneName=calf_r}
   }
   n25 = Anim Dynamics - Bone: dyn_r_u_calf_strap_04(ComponentPose=@n24, Alpha=1, ExternalForce="0,0,0") {
    angularSpringConstant=1800
    bAngularSpring=true
    bUseGravityOverride=true
    simulationSpace=RootRelative
    alphaBoolBlend {blendOption=Linear}
    boundBone {boneName=dyn_r_u_calf_strap_04}
    componentAppliedLinearAccClamp=<1000 1000 1000>
    componentLinearAccScale=<0.1 0.1 0.1>
    componentLinearVelScale=<0.1 0.1 0.1>
    gravityOverride {z=-980}
    physicsBodyDefinitions [{
      boundBone {boneName=dyn_r_u_calf_strap_04}
      boxExtents=<10 20 6>
      constraintSetup {angularLimitsMax=<2 5 3>, angularLimitsMin=<-2 -5 -3>, angularTarget=<10 0 0>}
      localJointOffset=<2 -2 0>
     }]
    relativeSpaceBone {boneName=calf_r}
   }
   n26 = Anim Dynamics - Bone: dyn_r_d_calf_strap_01(ComponentPose=@n25, Alpha=1, ExternalForce="0,0,0") {
    angularSpringConstant=1800
    bAngularSpring=true
    bUseGravityOverride=true
    simulationSpace=RootRelative
    alphaBoolBlend {blendOption=Linear}
    boundBone {boneName=dyn_r_d_calf_strap_01}
    componentAppliedLinearAccClamp=<1000 1000 1000>
    componentLinearAccScale=<0.1 0.1 0.1>
    componentLinearVelScale=<0.1 0.1 0.1>
    gravityOverride {z=-980}
    physicsBodyDefinitions [{
      boundBone {boneName=dyn_r_d_calf_strap_01}
      boxExtents=<10 20 6>
      constraintSetup {angularLimitsMax=<2 5 3>, angularLimitsMin=<-2 -5 -3>, angularTarget=<10 0 0>}
      localJointOffset=<2 -2 0>
     }]
    relativeSpaceBone {boneName=calf_r}
   }
   n27 = Anim Dynamics - Bone: dyn_r_d_calf_strap_02(ComponentPose=@n26, Alpha=1, ExternalForce="0,0,0") {
    angularSpringConstant=1800
    bAngularSpring=true
    bUseGravityOverride=true
    simulationSpace=RootRelative
    alphaBoolBlend {blendOption=Linear}
    boundBone {boneName=dyn_r_d_calf_strap_02}
    componentAppliedLinearAccClamp=<1000 1000 1000>
    componentLinearAccScale=<0.1 0.1 0.1>
    componentLinearVelScale=<0.1 0.1 0.1>
    gravityOverride {z=-980}
    physicsBodyDefinitions [{
      boundBone {boneName=dyn_r_d_calf_strap_02}
      boxExtents=<10 20 6>
      constraintSetup {angularLimitsMax=<2 5 3>, angularLimitsMin=<-2 -5 -3>, angularTarget=<10 0 0>}
      localJointOffset=<2 -2 0>
     }]
    relativeSpaceBone {boneName=calf_r}
   }
   n28 = Anim Dynamics - Bone: dyn_r_d_calf_strap_03(ComponentPose=@n27, Alpha=1, ExternalForce="0,0,0") {
    angularSpringConstant=1800
    bAngularSpring=true
    bUseGravityOverride=true
    simulationSpace=RootRelative
    alphaBoolBlend {blendOption=Linear}
    boundBone {boneName=dyn_r_d_calf_strap_03}
    componentAppliedLinearAccClamp=<1000 1000 1000>
    componentLinearAccScale=<0.1 0.1 0.1>
    componentLinearVelScale=<0.1 0.1 0.1>
    gravityOverride {z=-980}
    physicsBodyDefinitions [{
      boundBone {boneName=dyn_r_d_calf_strap_03}
      boxExtents=<10 20 6>
      constraintSetup {angularLimitsMax=<2 5 3>, angularLimitsMin=<-2 -5 -3>, angularTarget=<10 0 0>}
      localJointOffset=<2 -2 0>
     }]
    relativeSpaceBone {boneName=calf_r}
   }
   n29 = Anim Dynamics - Bone: dyn_r_d_calf_strap_04(ComponentPose=@n28, Alpha=1, ExternalForce="0,0,0") {
    angularSpringConstant=1800
    bAngularSpring=true
    bUseGravityOverride=true
    simulationSpace=RootRelative
    alphaBoolBlend {blendOption=Linear}
    boundBone {boneName=dyn_r_d_calf_strap_04}
    componentAppliedLinearAccClamp=<1000 1000 1000>
    componentLinearAccScale=<0.1 0.1 0.1>
    componentLinearVelScale=<0.1 0.1 0.1>
    gravityOverride {z=-980}
    physicsBodyDefinitions [{
      boundBone {boneName=dyn_r_d_calf_strap_04}
      boxExtents=<10 20 6>
      constraintSetup {angularLimitsMax=<2 5 3>, angularLimitsMin=<-2 -5 -3>, angularTarget=<10 0 0>}
      localJointOffset=<2 -2 0>
     }]
    relativeSpaceBone {boneName=calf_r}
   }
   n30 = Anim Dynamics - Bone: dyn_l_u_calf_strap_01(ComponentPose=@n29, Alpha=1, ExternalForce="0,0,0") {
    angularSpringConstant=1800
    bAngularSpring=true
    bUseGravityOverride=true
    simulationSpace=RootRelative
    alphaBoolBlend {blendOption=Linear}
    boundBone {boneName=dyn_l_u_calf_strap_01}
    componentAppliedLinearAccClamp=<1000 1000 1000>
    componentLinearAccScale=<0.1 0.1 0.1>
    componentLinearVelScale=<0.1 0.1 0.1>
    gravityOverride {z=-980}
    physicsBodyDefinitions [{
      boundBone {boneName=dyn_l_u_calf_strap_01}
      boxExtents=<10 20 6>
      constraintSetup {angularLimitsMax=<2 5 3>, angularLimitsMin=<-2 -5 -3>, angularTarget=<10 0 0>}
      localJointOffset=<2 -2 0>
     }]
    relativeSpaceBone {boneName=calf_l}
   }
   n31 = Anim Dynamics - Bone: dyn_l_u_calf_strap_02(ComponentPose=@n30, Alpha=1, ExternalForce="0,0,0") {
    angularSpringConstant=1800
    bAngularSpring=true
    bUseGravityOverride=true
    simulationSpace=RootRelative
    alphaBoolBlend {blendOption=Linear}
    boundBone {boneName=dyn_l_u_calf_strap_02}
    componentAppliedLinearAccClamp=<1000 1000 1000>
    componentLinearAccScale=<0.1 0.1 0.1>
    componentLinearVelScale=<0.1 0.1 0.1>
    gravityOverride {z=-980}
    physicsBodyDefinitions [{
      boundBone {boneName=dyn_l_u_calf_strap_02}
      boxExtents=<10 20 6>
      constraintSetup {angularLimitsMax=<2 5 3>, angularLimitsMin=<-2 -5 -3>, angularTarget=<10 0 0>}
      localJointOffset=<2 -2 0>
     }]
    relativeSpaceBone {boneName=calf_l}
   }
   n32 = Anim Dynamics - Bone: dyn_l_u_calf_strap_03(ComponentPose=@n31, Alpha=1, ExternalForce="0,0,0") {
    angularSpringConstant=1800
    bAngularSpring=true
    bUseGravityOverride=true
    simulationSpace=RootRelative
    alphaBoolBlend {blendOption=Linear}
    boundBone {boneName=dyn_l_u_calf_strap_03}
    componentAppliedLinearAccClamp=<1000 1000 1000>
    componentLinearAccScale=<0.1 0.1 0.1>
    componentLinearVelScale=<0.1 0.1 0.1>
    gravityOverride {z=-980}
    physicsBodyDefinitions [{
      boundBone {boneName=dyn_l_u_calf_strap_03}
      boxExtents=<10 20 6>
      constraintSetup {angularLimitsMax=<2 5 3>, angularLimitsMin=<-2 -5 -3>, angularTarget=<10 0 0>}
      localJointOffset=<2 -2 0>
     }]
    relativeSpaceBone {boneName=calf_l}
   }
   n33 = Anim Dynamics - Bone: dyn_l_u_calf_strap_04(ComponentPose=@n32, Alpha=1, ExternalForce="0,0,0") {
    angularSpringConstant=1800
    bAngularSpring=true
    bUseGravityOverride=true
    simulationSpace=RootRelative
    alphaBoolBlend {blendOption=Linear}
    boundBone {boneName=dyn_l_u_calf_strap_04}
    componentAppliedLinearAccClamp=<1000 1000 1000>
    componentLinearAccScale=<0.1 0.1 0.1>
    componentLinearVelScale=<0.1 0.1 0.1>
    gravityOverride {z=-980}
    physicsBodyDefinitions [{
      boundBone {boneName=dyn_l_u_calf_strap_04}
      boxExtents=<10 20 6>
      constraintSetup {angularLimitsMax=<2 5 3>, angularLimitsMin=<-2 -5 -3>, angularTarget=<10 0 0>}
      localJointOffset=<2 -2 0>
     }]
    relativeSpaceBone {boneName=calf_l}
   }
   n34 = Anim Dynamics - Bone: dyn_l_d_calf_strap_01(ComponentPose=@n33, Alpha=1, ExternalForce="0,0,0") {
    angularSpringConstant=1800
    bAngularSpring=true
    bUseGravityOverride=true
    simulationSpace=RootRelative
    alphaBoolBlend {blendOption=Linear}
    boundBone {boneName=dyn_l_d_calf_strap_01}
    componentAppliedLinearAccClamp=<1000 1000 1000>
    componentLinearAccScale=<0.1 0.1 0.1>
    componentLinearVelScale=<0.1 0.1 0.1>
    gravityOverride {z=-980}
    physicsBodyDefinitions [{
      boundBone {boneName=dyn_l_d_calf_strap_01}
      boxExtents=<10 20 6>
      constraintSetup {angularLimitsMax=<2 5 3>, angularLimitsMin=<-2 -5 -3>, angularTarget=<10 0 0>}
      localJointOffset=<2 -2 0>
     }]
    relativeSpaceBone {boneName=calf_l}
   }
   n35 = Anim Dynamics - Bone: dyn_l_d_calf_strap_02(ComponentPose=@n34, Alpha=1, ExternalForce="0,0,0") {
    angularSpringConstant=1800
    bAngularSpring=true
    bUseGravityOverride=true
    simulationSpace=RootRelative
    alphaBoolBlend {blendOption=Linear}
    boundBone {boneName=dyn_l_d_calf_strap_02}
    componentAppliedLinearAccClamp=<1000 1000 1000>
    componentLinearAccScale=<0.1 0.1 0.1>
    componentLinearVelScale=<0.1 0.1 0.1>
    gravityOverride {z=-980}
    physicsBodyDefinitions [{
      boundBone {boneName=dyn_l_d_calf_strap_02}
      boxExtents=<10 20 6>
      constraintSetup {angularLimitsMax=<2 5 3>, angularLimitsMin=<-2 -5 -3>, angularTarget=<10 0 0>}
      localJointOffset=<2 -2 0>
     }]
    relativeSpaceBone {boneName=calf_l}
   }
   n36 = Anim Dynamics - Bone: dyn_l_d_calf_strap_03(ComponentPose=@n35, Alpha=1, ExternalForce="0,0,0") {
    angularSpringConstant=1800
    bAngularSpring=true
    bUseGravityOverride=true
    simulationSpace=RootRelative
    alphaBoolBlend {blendOption=Linear}
    boundBone {boneName=dyn_l_d_calf_strap_03}
    componentAppliedLinearAccClamp=<1000 1000 1000>
    componentLinearAccScale=<0.1 0.1 0.1>
    componentLinearVelScale=<0.1 0.1 0.1>
    gravityOverride {z=-980}
    physicsBodyDefinitions [{
      boundBone {boneName=dyn_l_d_calf_strap_03}
      boxExtents=<10 20 6>
      constraintSetup {angularLimitsMax=<2 5 3>, angularLimitsMin=<-2 -5 -3>, angularTarget=<10 0 0>}
      localJointOffset=<2 -2 0>
     }]
    relativeSpaceBone {boneName=calf_l}
   }
   n37 = Anim Dynamics - Bone: dyn_l_d_calf_strap_04(ComponentPose=@n36, Alpha=1, ExternalForce="0,0,0") {
    angularSpringConstant=1800
    bAngularSpring=true
    bUseGravityOverride=true
    simulationSpace=RootRelative
    alphaBoolBlend {blendOption=Linear}
    boundBone {boneName=dyn_l_d_calf_strap_04}
    componentAppliedLinearAccClamp=<1000 1000 1000>
    componentLinearAccScale=<0.1 0.1 0.1>
    componentLinearVelScale=<0.1 0.1 0.1>
    gravityOverride {z=-980}
    physicsBodyDefinitions [{
      boundBone {boneName=dyn_l_d_calf_strap_04}
      boxExtents=<10 20 6>
      constraintSetup {angularLimitsMax=<2 5 3>, angularLimitsMin=<-2 -5 -3>, angularTarget=<10 0 0>}
      localJointOffset=<2 -2 0>
     }]
    relativeSpaceBone {boneName=calf_l}
   }
   n38 = Anim Dynamics - Bone: dyn_r_back_buck_01(ComponentPose=@n37, Alpha=1, ExternalForce="0,0,0") {
    angularSpringConstant=500
    bAngularSpring=true
    bUseGravityOverride=true
    simulationSpace=RootRelative
    alphaBoolBlend {blendOption=Linear}
    boundBone {boneName=dyn_r_back_buck_01}
    componentAppliedLinearAccClamp=<1000 1000 1000>
    componentLinearAccScale=<0.1 0.1 0.1>
    componentLinearVelScale=<0.1 0.1 0.1>
    gravityOverride {z=-980}
    physicsBodyDefinitions [{
      boundBone {boneName=dyn_r_back_buck_01}
      boxExtents=<16 20 20>
      constraintSetup {angularLimitsMax=<1 3 8>, angularLimitsMin=<-1 -3 -12>, angularTarget=<10 0 0>}
      localJointOffset=<3 0 0.5>
     }]
    relativeSpaceBone {boneName=clavicle_r}
   }
   n39 = Anim Dynamics - Bone: dyn_r_chest_buck_01(ComponentPose=@n38, Alpha=1, ExternalForce="0,0,0") {
    angularSpringConstant=500
    bAngularSpring=true
    bUseGravityOverride=true
    simulationSpace=RootRelative
    alphaBoolBlend {blendOption=Linear}
    boundBone {boneName=dyn_r_chest_buck_01}
    componentAppliedLinearAccClamp=<1000 1000 1000>
    componentLinearAccScale=<0.1 0.1 0.1>
    componentLinearVelScale=<0.1 0.1 0.1>
    gravityOverride {z=-980}
    physicsBodyDefinitions [{
      boundBone {boneName=dyn_r_chest_buck_01}
      boxExtents=<10 12 20>
      constraintSetup {angularLimitsMax=<1 3 15>, angularLimitsMin=<-1 -6 -5>, angularTarget=<10 0 0>}
      localJointOffset=<1 0 0>
     }]
    relativeSpaceBone {boneName=spine_05}
   }
   n40 = Anim Dynamics - Bone: dyn_r_chest_buck_02(ComponentPose=@n39, Alpha=1, ExternalForce="0,0,0") {
    angularSpringConstant=500
    bAngularSpring=true
    bUseGravityOverride=true
    simulationSpace=RootRelative
    alphaBoolBlend {blendOption=Linear}
    boundBone {boneName=dyn_r_chest_buck_02}
    componentAppliedLinearAccClamp=<1000 1000 1000>
    componentLinearAccScale=<0.1 0.1 0.1>
    componentLinearVelScale=<0.1 0.1 0.1>
    gravityOverride {z=-980}
    physicsBodyDefinitions [{
      boundBone {boneName=dyn_r_chest_buck_02}
      boxExtents=<10 12 20>
      constraintSetup {angularLimitsMax=<1 3 15>, angularLimitsMin=<-1 -6 -5>, angularTarget=<10 0 0>}
      localJointOffset=<1 0 0>
     }]
    relativeSpaceBone {boneName=spine_05}
   }
   n41 = Anim Dynamics - Bone: dyn_r_chest_buck_03(ComponentPose=@n40, Alpha=1, ExternalForce="0,0,0") {
    angularSpringConstant=500
    bAngularSpring=true
    bUseGravityOverride=true
    simulationSpace=RootRelative
    alphaBoolBlend {blendOption=Linear}
    boundBone {boneName=dyn_r_chest_buck_03}
    componentAppliedLinearAccClamp=<1000 1000 1000>
    componentLinearAccScale=<0.1 0.1 0.1>
    componentLinearVelScale=<0.1 0.1 0.1>
    gravityOverride {z=-980}
    physicsBodyDefinitions [{
      boundBone {boneName=dyn_r_chest_buck_03}
      boxExtents=<10 12 20>
      constraintSetup {angularLimitsMax=<1 3 15>, angularLimitsMin=<-1 -6 -5>, angularTarget=<10 0 0>}
      localJointOffset=<1 0 0>
     }]
    relativeSpaceBone {boneName=spine_05}
   }
   n42 = Anim Dynamics - Bone: dyn_l_back_buck_01(ComponentPose=@n41, Alpha=1, ExternalForce="0,0,0") {
    angularSpringConstant=500
    bAngularSpring=true
    bUseGravityOverride=true
    simulationSpace=RootRelative
    alphaBoolBlend {blendOption=Linear}
    boundBone {boneName=dyn_l_back_buck_01}
    componentAppliedLinearAccClamp=<1000 1000 1000>
    componentLinearAccScale=<0.1 0.1 0.1>
    componentLinearVelScale=<0.1 0.1 0.1>
    gravityOverride {z=-980}
    physicsBodyDefinitions [{
      boundBone {boneName=dyn_l_back_buck_01}
      boxExtents=<16 20 20>
      constraintSetup {angularLimitsMax=<1 3 8>, angularLimitsMin=<-1 -3 -12>, angularTarget=<10 0 0>}
      localJointOffset=<3 0 0.5>
     }]
    relativeSpaceBone {boneName=clavicle_r}
   }
   n43 = Anim Dynamics - Bone: dyn_l_chest_buck_01(ComponentPose=@n42, Alpha=1, ExternalForce="0,0,0") {
    angularSpringConstant=500
    bAngularSpring=true
    bUseGravityOverride=true
    simulationSpace=RootRelative
    alphaBoolBlend {blendOption=Linear}
    boundBone {boneName=dyn_l_chest_buck_01}
    componentAppliedLinearAccClamp=<1000 1000 1000>
    componentLinearAccScale=<0.1 0.1 0.1>
    componentLinearVelScale=<0.1 0.1 0.1>
    gravityOverride {z=-980}
    physicsBodyDefinitions [{
      boundBone {boneName=dyn_l_chest_buck_01}
      boxExtents=<10 12 20>
      constraintSetup {angularLimitsMax=<1 6 15>, angularLimitsMin=<-1 -3 -5>, angularTarget=<10 0 0>}
      localJointOffset=<1 0 0>
     }]
    relativeSpaceBone {boneName=spine_05}
   }
   n44 = Anim Dynamics - Bone: dyn_l_chest_buck_02(ComponentPose=@n43, Alpha=1, ExternalForce="0,0,0") {
    angularSpringConstant=500
    bAngularSpring=true
    bUseGravityOverride=true
    simulationSpace=RootRelative
    alphaBoolBlend {blendOption=Linear}
    boundBone {boneName=dyn_l_chest_buck_02}
    componentAppliedLinearAccClamp=<1000 1000 1000>
    componentLinearAccScale=<0.1 0.1 0.1>
    componentLinearVelScale=<0.1 0.1 0.1>
    gravityOverride {z=-980}
    physicsBodyDefinitions [{
      boundBone {boneName=dyn_l_chest_buck_02}
      boxExtents=<10 12 20>
      constraintSetup {angularLimitsMax=<1 6 15>, angularLimitsMin=<-1 -3 -5>, angularTarget=<10 0 0>}
      localJointOffset=<1 0 0>
     }]
    relativeSpaceBone {boneName=spine_05}
   }
   n45 = Anim Dynamics - Bone: dyn_l_chest_buck_03(ComponentPose=@n44, Alpha=1, ExternalForce="0,0,0") {
    angularSpringConstant=500
    bAngularSpring=true
    bUseGravityOverride=true
    simulationSpace=RootRelative
    alphaBoolBlend {blendOption=Linear}
    boundBone {boneName=dyn_l_chest_buck_03}
    componentAppliedLinearAccClamp=<1000 1000 1000>
    componentLinearAccScale=<0.1 0.1 0.1>
    componentLinearVelScale=<0.1 0.1 0.1>
    gravityOverride {z=-980}
    physicsBodyDefinitions [{
      boundBone {boneName=dyn_l_chest_buck_03}
      boxExtents=<10 12 20>
      constraintSetup {angularLimitsMax=<1 6 15>, angularLimitsMin=<-1 -3 -5>, angularTarget=<10 0 0>}
      localJointOffset=<1 0 0>
     }]
    relativeSpaceBone {boneName=spine_05}
   }
   n46 = Get AnimDynaAlpha() -> (AnimDynaAlpha:double)
   n47 = Get AnimSpring() -> (AnimSpring:double)
   n48 = Anim Dynamics - Bone: dyn_fluff_01!(ComponentPose=@n45, Alpha=@n46, AngularSpringConstant:float=@n47, ExternalForce="0,0,0") -> (Pose) {
    angularSpringConstant=2750
    bAngularSpring=true
    bUseGravityOverride=true
    simulationSpace=BoneRelative
    alphaBoolBlend {blendOption=Linear}
    boundBone {boneName=dyn_fluff_01}
    componentAppliedLinearAccClamp=<1150 1150 1150>
    componentLinearAccScale=<0.1 0.1 1>
    componentLinearVelScale=<0.1 0.1 0.5>
    gravityOverride {z=-980}
    physicsBodyDefinitions [{
      boundBone {boneName=dyn_fluff_01}
      boxExtents=<20 20 20>
      constraintSetup {angularLimitsMax=<1 10 6>, angularLimitsMin=<-1 -20 -6>, angularTarget=<1 0 -2>}
      localJointOffset=<10 0 12>
     }]
    relativeSpaceBone {boneName=root}
   }
   n49 = Anim Dynamics - Bone: dyn_fluff_02!(ComponentPose=@n48, Alpha=@n46, AngularSpringConstant:float=@n47, ExternalForce="0,0,0") -> (Pose) {
    angularSpringConstant=2750
    bAngularSpring=true
    bUseGravityOverride=true
    simulationSpace=BoneRelative
    alphaBoolBlend {blendOption=Linear}
    boundBone {boneName=dyn_fluff_02}
    componentAppliedLinearAccClamp=<1000 1000 1000>
    componentLinearAccScale=<0.1 0.1 1>
    componentLinearVelScale=<0.1 0.1 0.5>
    gravityOverride {z=-980}
    physicsBodyDefinitions [{
      boundBone {boneName=dyn_fluff_02}
      boxExtents=<28 28 28>
      constraintSetup {angularLimitsMax=<1 10 6>, angularLimitsMin=<-1 -20 -6>, angularTarget=<1 0 -2>}
      localJointOffset=<10 0 12>
     }]
    relativeSpaceBone {boneName=root}
   }
   n50 = Anim Dynamics - Bone: dyn_fluff_03!(ComponentPose=@n49, Alpha=@n46, AngularSpringConstant:float=@n47, ExternalForce="0,0,0") -> (Pose) {
    angularSpringConstant=2750
    bAngularSpring=true
    bUseGravityOverride=true
    simulationSpace=BoneRelative
    alphaBoolBlend {blendOption=Linear}
    boundBone {boneName=dyn_fluff_03}
    componentAppliedLinearAccClamp=<1150 1150 1150>
    componentLinearAccScale=<0.1 0.1 1>
    componentLinearVelScale=<0.1 0.1 0.5>
    gravityOverride {z=-980}
    physicsBodyDefinitions [{
      boundBone {boneName=dyn_fluff_03}
      boxExtents=<25 25 25>
      constraintSetup {angularLimitsMax=<1 10 6>, angularLimitsMin=<-1 -20 -6>, angularTarget=<1 0 -2>}
      localJointOffset=<8 0 12>
     }]
    relativeSpaceBone {boneName=root}
   }
   n51 = Anim Dynamics - Bone: dyn_fluff_04!(ComponentPose=@n50, Alpha=@n46, AngularSpringConstant:float=@n47, ExternalForce="0,0,0") -> (Pose) {
    angularSpringConstant=2750
    bAngularSpring=true
    bUseGravityOverride=true
    simulationSpace=BoneRelative
    alphaBoolBlend {blendOption=Linear}
    boundBone {boneName=dyn_fluff_04}
    componentAppliedLinearAccClamp=<1000 1000 1000>
    componentLinearAccScale=<0.1 0.1 1>
    componentLinearVelScale=<0.1 0.1 0.5>
    gravityOverride {z=-980}
    physicsBodyDefinitions [{
      boundBone {boneName=dyn_fluff_04}
      boxExtents=<20 20 20>
      constraintSetup {angularLimitsMax=<1 10 6>, angularLimitsMin=<-1 -20 -6>, angularTarget=<1 0 -2>}
      localJointOffset=<10 0 12>
     }]
    relativeSpaceBone {boneName=root}
   }
   n52 = Anim Dynamics - Bone: dyn_fluff_05!(ComponentPose=@n51, Alpha=@n46, AngularSpringConstant:float=@n47, ExternalForce="0,0,0") -> (Pose) {
    angularSpringConstant=2750
    bAngularSpring=true
    bUseGravityOverride=true
    simulationSpace=BoneRelative
    alphaBoolBlend {blendOption=Linear}
    boundBone {boneName=dyn_fluff_05}
    componentAppliedLinearAccClamp=<1150 1150 1150>
    componentLinearAccScale=<0.1 0.1 1>
    componentLinearVelScale=<0.1 0.1 0.5>
    gravityOverride {z=-980}
    physicsBodyDefinitions [{
      boundBone {boneName=dyn_fluff_05}
      boxExtents=<28 28 28>
      constraintSetup {angularLimitsMax=<1 10 6>, angularLimitsMin=<-1 -20 -6>, angularTarget=<1 0 -2>}
      localJointOffset=<10 0 12>
     }]
    relativeSpaceBone {boneName=root}
   }
   n53 = Anim Dynamics - Bone: dyn_fluff_06!(ComponentPose=@n52, Alpha=@n46, AngularSpringConstant:float=@n47, ExternalForce="0,0,0") -> (Pose) {
    angularSpringConstant=2750
    bAngularSpring=true
    bUseGravityOverride=true
    simulationSpace=BoneRelative
    alphaBoolBlend {blendOption=Linear}
    boundBone {boneName=dyn_fluff_06}
    componentAppliedLinearAccClamp=<1000 1000 1000>
    componentLinearAccScale=<0.1 0.1 1>
    componentLinearVelScale=<0.1 0.1 0.5>
    gravityOverride {z=-980}
    physicsBodyDefinitions [{
      boundBone {boneName=dyn_fluff_06}
      boxExtents=<25 25 25>
      constraintSetup {angularLimitsMax=<1 10 6>, angularLimitsMin=<-1 -20 -6>, angularTarget=<1 0 -2>}
      localJointOffset=<8 0 12>
     }]
    relativeSpaceBone {boneName=root}
   }
   n54 = Anim Dynamics - Bone: dyn_fluff_07!(ComponentPose=@n53, Alpha=@n46, AngularSpringConstant:float=@n47, ExternalForce="0,0,0") -> (Pose) {
    angularSpringConstant=2750
    bAngularSpring=true
    bUseGravityOverride=true
    simulationSpace=BoneRelative
    alphaBoolBlend {blendOption=Linear}
    boundBone {boneName=dyn_fluff_07}
    componentAppliedLinearAccClamp=<1000 1000 1000>
    componentLinearAccScale=<0.1 0.1 1>
    componentLinearVelScale=<0.1 0.1 0.5>
    gravityOverride {z=-980}
    physicsBodyDefinitions [{
      boundBone {boneName=dyn_fluff_07}
      boxExtents=<25 25 25>
      constraintSetup {angularLimitsMax=<1 10 6>, angularLimitsMin=<-1 -20 -6>, angularTarget=<1 0 -2>}
      localJointOffset=<10 0 12>
     }]
    relativeSpaceBone {boneName=root}
   }
   n55 = Anim Dynamics - Bone: dyn_fluff_08!(ComponentPose=@n54, Alpha=@n46, AngularSpringConstant:float=@n47, ExternalForce="0,0,0") -> (Pose) {
    angularSpringConstant=2750
    bAngularSpring=true
    bUseGravityOverride=true
    simulationSpace=BoneRelative
    alphaBoolBlend {blendOption=Linear}
    boundBone {boneName=dyn_fluff_08}
    componentAppliedLinearAccClamp=<1150 1150 1150>
    componentLinearAccScale=<0.1 0.1 1>
    componentLinearVelScale=<0.1 0.1 0.5>
    gravityOverride {z=-980}
    physicsBodyDefinitions [{
      boundBone {boneName=dyn_fluff_08}
      boxExtents=<20 20 20>
      constraintSetup {angularLimitsMax=<1 10 6>, angularLimitsMin=<-1 -20 -6>, angularTarget=<1 0 -2>}
      localJointOffset=<2 0 12>
     }]
    relativeSpaceBone {boneName=root}
   }
   n56 = Anim Dynamics - Bone: dyn_fluff_09!(ComponentPose=@n55, Alpha=@n46, AngularSpringConstant:float=@n47, ExternalForce="0,0,0") -> (Pose) {
    angularSpringConstant=2750
    bAngularSpring=true
    bUseGravityOverride=true
    simulationSpace=BoneRelative
    alphaBoolBlend {blendOption=Linear}
    boundBone {boneName=dyn_fluff_09}
    componentAppliedLinearAccClamp=<1000 1000 1000>
    componentLinearAccScale=<0.1 0.1 1>
    componentLinearVelScale=<0.1 0.1 0.5>
    gravityOverride {z=-980}
    physicsBodyDefinitions [{
      boundBone {boneName=dyn_fluff_09}
      boxExtents=<28 28 28>
      constraintSetup {angularLimitsMax=<1 10 6>, angularLimitsMin=<-1 -20 -6>, angularTarget=<1 0 -2>}
      localJointOffset=<10 0 12>
     }]
    relativeSpaceBone {boneName=root}
   }
   n57 = Anim Dynamics - Bone: dyn_fluff_10!(ComponentPose=@n56, Alpha=@n46, AngularSpringConstant:float=@n47, ExternalForce="0,0,0") -> (Pose) {
    angularSpringConstant=2750
    bAngularSpring=true
    bUseGravityOverride=true
    simulationSpace=BoneRelative
    alphaBoolBlend {blendOption=Linear}
    boundBone {boneName=dyn_fluff_10}
    componentAppliedLinearAccClamp=<1150 1150 1150>
    componentLinearAccScale=<0.1 0.1 1>
    componentLinearVelScale=<0.1 0.1 0.5>
    gravityOverride {z=-980}
    physicsBodyDefinitions [{
      boundBone {boneName=dyn_fluff_10}
      boxExtents=<25 25 25>
      constraintSetup {angularLimitsMax=<1 10 6>, angularLimitsMin=<-1 -20 -6>, angularTarget=<1 0 -2>}
      localJointOffset=<8 0 12>
     }]
    relativeSpaceBone {boneName=root}
   }
   n58 = Anim Dynamics - Bone: dyn_fluff_11!(ComponentPose=@n57, Alpha=@n46, AngularSpringConstant:float=@n47, ExternalForce="0,0,0") -> (Pose) {
    angularSpringConstant=2750
    bAngularSpring=true
    bUseGravityOverride=true
    simulationSpace=BoneRelative
    alphaBoolBlend {blendOption=Linear}
    boundBone {boneName=dyn_fluff_11}
    componentAppliedLinearAccClamp=<1000 1000 1000>
    componentLinearAccScale=<0.1 0.1 1>
    componentLinearVelScale=<0.1 0.1 0.5>
    gravityOverride {z=-980}
    physicsBodyDefinitions [{
      boundBone {boneName=dyn_fluff_11}
      boxExtents=<20 20 20>
      constraintSetup {angularLimitsMax=<1 10 6>, angularLimitsMin=<-1 -20 -6>, angularTarget=<1 0 -2>}
      localJointOffset=<10 0 12>
     }]
    relativeSpaceBone {boneName=root}
   }
   n59 = Anim Dynamics - Bone: dyn_fluff_11(ComponentPose=@n58, Alpha=@n46, AngularSpringConstant=@n47, ExternalForce="0,0,0") {
    angularSpringConstant=2750
    bAngularSpring=true
    bUseGravityOverride=true
    simulationSpace=BoneRelative
    alphaBoolBlend {blendOption=Linear}
    boundBone {boneName=dyn_fluff_11}
    componentAppliedLinearAccClamp=<1150 1150 1150>
    componentLinearAccScale=<0.1 0.1 1>
    componentLinearVelScale=<0.1 0.1 0.5>
    gravityOverride {z=-980}
    physicsBodyDefinitions [{
      boundBone {boneName=dyn_fluff_11}
      boxExtents=<28 28 28>
      constraintSetup {angularLimitsMax=<1 10 6>, angularLimitsMin=<-1 -20 -2>, angularTarget=<1 0 -2>}
      localJointOffset=<2 0 12>
     }]
    relativeSpaceBone {boneName=root}
   }
   n60 = Anim Dynamics - Bone: dyn_fluff_12!(ComponentPose=@n59, Alpha=@n46, AngularSpringConstant:float=@n47, ExternalForce="0,0,0") -> (Pose) {
    angularSpringConstant=2750
    bAngularSpring=true
    bUseGravityOverride=true
    simulationSpace=BoneRelative
    alphaBoolBlend {blendOption=Linear}
    boundBone {boneName=dyn_fluff_12}
    componentAppliedLinearAccClamp=<1000 1000 1000>
    componentLinearAccScale=<0.1 0.1 1>
    componentLinearVelScale=<0.1 0.1 0.5>
    gravityOverride {z=-980}
    physicsBodyDefinitions [{
      boundBone {boneName=dyn_fluff_12}
      boxExtents=<25 25 25>
      constraintSetup {angularLimitsMax=<1 10 6>, angularLimitsMin=<-1 -20 -6>, angularTarget=<1 0 -2>}
      localJointOffset=<8 0 12>
     }]
    relativeSpaceBone {boneName=root}
   }
   n61 = Anim Dynamics - Bone: dyn_fluff_13!(ComponentPose=@n60, Alpha=@n46, AngularSpringConstant:float=@n47, ExternalForce="0,0,0") -> (Pose) {
    angularSpringConstant=2750
    bAngularSpring=true
    bUseGravityOverride=true
    simulationSpace=BoneRelative
    alphaBoolBlend {blendOption=Linear}
    boundBone {boneName=dyn_fluff_13}
    componentAppliedLinearAccClamp=<1150 1150 1150>
    componentLinearAccScale=<0.1 0.1 1>
    componentLinearVelScale=<0.1 0.1 0.5>
    gravityOverride {z=-980}
    physicsBodyDefinitions [{
      boundBone {boneName=dyn_fluff_13}
      boxExtents=<20 20 20>
      constraintSetup {angularLimitsMax=<1 10 6>, angularLimitsMin=<-1 -20 -6>, angularTarget=<1 0 -2>}
      localJointOffset=<2 0 12>
     }]
    relativeSpaceBone {boneName=root}
   }
   n62 = Anim Dynamics - Bone: dyn_fluff_14!(ComponentPose=@n61, Alpha=@n46, AngularSpringConstant:float=@n47, ExternalForce="0,0,0") -> (Pose) {
    angularSpringConstant=2750
    bAngularSpring=true
    bUseGravityOverride=true
    simulationSpace=BoneRelative
    alphaBoolBlend {blendOption=Linear}
    boundBone {boneName=dyn_fluff_14}
    componentAppliedLinearAccClamp=<1000 1000 1000>
    componentLinearAccScale=<0.1 0.1 1>
    componentLinearVelScale=<0.1 0.1 0.5>
    gravityOverride {z=-980}
    physicsBodyDefinitions [{
      boundBone {boneName=dyn_fluff_14}
      boxExtents=<28 28 28>
      constraintSetup {angularLimitsMax=<1 10 6>, angularLimitsMin=<-1 -20 -6>, angularTarget=<1 0 -2>}
      localJointOffset=<10 0 12>
     }]
    relativeSpaceBone {boneName=root}
   }
   n63 = Anim Dynamics - Bone: dyn_fluff_15!(ComponentPose=@n62, Alpha=@n46, AngularSpringConstant:float=@n47, ExternalForce="0,0,0") -> (Pose) {
    angularSpringConstant=2750
    bAngularSpring=true
    bUseGravityOverride=true
    simulationSpace=BoneRelative
    alphaBoolBlend {blendOption=Linear}
    boundBone {boneName=dyn_fluff_15}
    componentAppliedLinearAccClamp=<1150 1150 1150>
    componentLinearAccScale=<0.1 0.1 1>
    componentLinearVelScale=<0.1 0.1 0.5>
    gravityOverride {z=-980}
    physicsBodyDefinitions [{
      boundBone {boneName=dyn_fluff_15}
      boxExtents=<25 25 25>
      constraintSetup {angularLimitsMax=<1 10 6>, angularLimitsMin=<-1 -20 -6>, angularTarget=<1 0 -2>}
      localJointOffset=<8 0 12>
     }]
    relativeSpaceBone {boneName=root}
   }
   n64 = Anim Dynamics - Bone: dyn_fluff_16!(ComponentPose=@n63, Alpha=@n46, AngularSpringConstant:float=@n47, ExternalForce="0,0,0") -> (Pose) {
    angularSpringConstant=2750
    bAngularSpring=true
    bUseGravityOverride=true
    simulationSpace=BoneRelative
    alphaBoolBlend {blendOption=Linear}
    boundBone {boneName=dyn_fluff_16}
    componentAppliedLinearAccClamp=<1000 1000 1000>
    componentLinearAccScale=<0.1 0.1 1>
    componentLinearVelScale=<0.1 0.1 0.5>
    gravityOverride {z=-980}
    physicsBodyDefinitions [{
      boundBone {boneName=dyn_fluff_16}
      boxExtents=<20 20 20>
      constraintSetup {angularLimitsMax=<1 10 6>, angularLimitsMin=<-1 -20 -6>, angularTarget=<1 0 -2>}
      localJointOffset=<2 0 12>
     }]
    relativeSpaceBone {boneName=root}
   }
   n65 = Anim Dynamics - Bone: dyn_fluff_17!(ComponentPose=@n64, Alpha=@n46, AngularSpringConstant:float=@n47, ExternalForce="0,0,0") -> (Pose) {
    angularSpringConstant=2750
    bAngularSpring=true
    bUseGravityOverride=true
    simulationSpace=BoneRelative
    alphaBoolBlend {blendOption=Linear}
    boundBone {boneName=dyn_fluff_17}
    componentAppliedLinearAccClamp=<1150 1150 1150>
    componentLinearAccScale=<0.1 0.1 1>
    componentLinearVelScale=<0.1 0.1 0.5>
    gravityOverride {z=-980}
    physicsBodyDefinitions [{
      boundBone {boneName=dyn_fluff_17}
      boxExtents=<28 28 28>
      constraintSetup {angularLimitsMax=<1 10 6>, angularLimitsMin=<-1 -20 -6>, angularTarget=<1 0 -2>}
      localJointOffset=<10 0 12>
     }]
    relativeSpaceBone {boneName=root}
   }
   n66 = Anim Dynamics - Bone: dyn_fluff_18!(ComponentPose=@n65, Alpha=@n46, AngularSpringConstant:float=@n47, ExternalForce="0,0,0") -> (Pose) {
    angularSpringConstant=2750
    bAngularSpring=true
    bUseGravityOverride=true
    simulationSpace=BoneRelative
    alphaBoolBlend {blendOption=Linear}
    boundBone {boneName=dyn_fluff_18}
    componentAppliedLinearAccClamp=<1000 1000 1000>
    componentLinearAccScale=<0.1 0.1 1>
    componentLinearVelScale=<0.1 0.1 0.5>
    gravityOverride {z=-980}
    physicsBodyDefinitions [{
      boundBone {boneName=dyn_fluff_18}
      boxExtents=<25 25 25>
      constraintSetup {angularLimitsMax=<1 10 6>, angularLimitsMin=<-1 -20 -6>, angularTarget=<1 0 -2>}
      localJointOffset=<8 0 12>
     }]
    relativeSpaceBone {boneName=root}
   }
   n67 = Anim Dynamics - Bone: dyn_fluff_19!(ComponentPose=@n66, Alpha=@n46, AngularSpringConstant:float=@n47, ExternalForce="0,0,0") -> (Pose) {
    angularSpringConstant=2750
    bAngularSpring=true
    bUseGravityOverride=true
    simulationSpace=BoneRelative
    alphaBoolBlend {blendOption=Linear}
    boundBone {boneName=dyn_fluff_19}
    componentAppliedLinearAccClamp=<1150 1150 1150>
    componentLinearAccScale=<0.1 0.1 1>
    componentLinearVelScale=<0.1 0.1 0.5>
    gravityOverride {z=-980}
    physicsBodyDefinitions [{
      boundBone {boneName=dyn_fluff_19}
      boxExtents=<25 25 25>
      constraintSetup {angularLimitsMax=<1 10 6>, angularLimitsMin=<-1 -20 -6>, angularTarget=<1 0 -2>}
      localJointOffset=<2 0 12>
     }]
    relativeSpaceBone {boneName=root}
   }
   n68 = Anim Dynamics - Bone: dyn_fluff_20!(ComponentPose=@n67, Alpha=@n46, AngularSpringConstant:float=@n47, ExternalForce="0,0,0") -> (Pose) {
    angularSpringConstant=2750
    bAngularSpring=true
    bUseGravityOverride=true
    simulationSpace=BoneRelative
    alphaBoolBlend {blendOption=Linear}
    boundBone {boneName=dyn_fluff_20}
    componentAppliedLinearAccClamp=<1000 1000 1000>
    componentLinearAccScale=<0.1 0.1 1>
    componentLinearVelScale=<0.1 0.1 0.5>
    gravityOverride {z=-980}
    physicsBodyDefinitions [{
      boundBone {boneName=dyn_fluff_20}
      boxExtents=<25 25 25>
      constraintSetup {angularLimitsMax=<1 10 6>, angularLimitsMin=<-1 -20 -6>, angularTarget=<1 0 -2>}
      localJointOffset=<10 0 12>
     }]
    relativeSpaceBone {boneName=root}
   }
   n69 = Anim Dynamics - Bone: dyn_fluff_21!(ComponentPose=@n68, Alpha=@n46, AngularSpringConstant:float=@n47, ExternalForce="0,0,0") -> (Pose) {
    angularSpringConstant=2750
    bAngularSpring=true
    bUseGravityOverride=true
    simulationSpace=BoneRelative
    alphaBoolBlend {blendOption=Linear}
    boundBone {boneName=dyn_fluff_21}
    componentAppliedLinearAccClamp=<1150 1150 1150>
    componentLinearAccScale=<0.1 0.1 1>
    componentLinearVelScale=<0.1 0.1 0.5>
    gravityOverride {z=-980}
    physicsBodyDefinitions [{
      boundBone {boneName=dyn_fluff_21}
      boxExtents=<20 20 20>
      constraintSetup {angularLimitsMax=<1 10 6>, angularLimitsMin=<-1 -20 -6>, angularTarget=<1 0 -2>}
      localJointOffset=<10 0 12>
     }]
    relativeSpaceBone {boneName=root}
   }
   n70 = Anim Dynamics - Bone: dyn_fluff_22!(ComponentPose=@n69, Alpha=@n46, AngularSpringConstant:float=@n47, ExternalForce="0,0,0") -> (Pose) {
    angularSpringConstant=2750
    bAngularSpring=true
    bUseGravityOverride=true
    simulationSpace=BoneRelative
    alphaBoolBlend {blendOption=Linear}
    boundBone {boneName=dyn_fluff_22}
    componentAppliedLinearAccClamp=<1000 1000 1000>
    componentLinearAccScale=<0.1 0.1 1>
    componentLinearVelScale=<0.1 0.1 0.5>
    gravityOverride {z=-980}
    physicsBodyDefinitions [{
      boundBone {boneName=dyn_fluff_22}
      boxExtents=<28 28 28>
      constraintSetup {angularLimitsMax=<1 10 6>, angularLimitsMin=<-1 -20 -6>, angularTarget=<1 0 -2>}
      localJointOffset=<10 0 12>
     }]
    relativeSpaceBone {boneName=root}
   }
   n71 = Anim Dynamics - Bone: dyn_fluff_23!(ComponentPose=@n70, Alpha=@n46, AngularSpringConstant:float=@n47, ExternalForce="0,0,0") -> (Pose) {
    angularSpringConstant=2750
    bAngularSpring=true
    bUseGravityOverride=true
    simulationSpace=BoneRelative
    alphaBoolBlend {blendOption=Linear}
    boundBone {boneName=dyn_fluff_23}
    componentAppliedLinearAccClamp=<1150 1150 1150>
    componentLinearAccScale=<0.1 0.1 1>
    componentLinearVelScale=<0.1 0.1 0.5>
    gravityOverride {z=-980}
    physicsBodyDefinitions [{
      boundBone {boneName=dyn_fluff_23}
      boxExtents=<25 25 25>
      constraintSetup {angularLimitsMax=<1 10 6>, angularLimitsMin=<-1 -20 -6>, angularTarget=<1 0 -2>}
      localJointOffset=<8 0 12>
     }]
    relativeSpaceBone {boneName=root}
   }
   n72 = Anim Dynamics - Bone: dyn_fluff_24!(ComponentPose=@n71, Alpha=@n46, AngularSpringConstant:float=@n47, ExternalForce="0,0,0") -> (Pose) {
    angularSpringConstant=2750
    bAngularSpring=true
    bUseGravityOverride=true
    simulationSpace=BoneRelative
    alphaBoolBlend {blendOption=Linear}
    boundBone {boneName=dyn_fluff_24}
    componentAppliedLinearAccClamp=<1000 1000 1000>
    componentLinearAccScale=<0.1 0.1 1>
    componentLinearVelScale=<0.1 0.1 0.5>
    gravityOverride {z=-980}
    physicsBodyDefinitions [{
      boundBone {boneName=dyn_fluff_24}
      boxExtents=<20 20 20>
      constraintSetup {angularLimitsMax=<1 10 6>, angularLimitsMin=<-1 -20 -6>, angularTarget=<1 0 -2>}
      localJointOffset=<10 0 12>
     }]
    relativeSpaceBone {boneName=root}
   }
   n73 = "Transform (Modify) Bone - Bone: dyn_l_f_strap_MSTR"(ComponentPose=@n72, Alpha=1, Translation="0,0,-0.2", Rotation="0,0,0", Scale="1,1,1") {
    rotationMode=BMM_Additive
    translationMode=BMM_Additive
    alphaBoolBlend {blendOption=Linear}
    boneToModify {boneName=dyn_l_f_strap_MSTR}
    translation {z=-0.2}
   }
   n74 = "Transform (Modify) Bone - Bone: dyn_l_f_strap_MSTR"(ComponentPose=@n73, Alpha=Get IsDBNO() -> (IsDBNO:double), Translation="1.856,5.040,1.155", Rotation="0,0,0", Scale="1,1,1") {
    rotationMode=BMM_Additive
    translationMode=BMM_Additive
    alphaBoolBlend {blendOption=Linear}
    boneToModify {boneName=dyn_l_f_strap_MSTR}
    translation=<1.856 5.04 1.155>
   }
   n75 = RigidBody(ComponentPose=@n74) {
    bOverrideWorldGravity=true
    overridePhysicsAsset=&/M_MED_BiruFang_Physics
    simulationSpace=BaseBoneSpace
    alphaBoolBlend {blendOption=Linear}
    baseBoneRef {boneName=pelvis}
    componentAppliedLinearAccClamp=<3000 3000 3000>
    componentLinearAccScale=<0.37 0.2 0.2>
    componentLinearVelScale=<0.2 0.1 0.1>
    overrideWorldGravity {z=-980}
    simSpaceSettings {
     maxAngularAcceleration=1000
     maxAngularVelocity=3
     maxLinearAcceleration=700
     maxLinearVelocity=500
     velocityScaleZ=0
     worldAlpha=1
     externalLinearDragV=<0.2 0.2>
    }
   }
   n76 = RigidBody(ComponentPose=@n75) {
    bOverrideWorldGravity=true
    overridePhysicsAsset=&/M_MED_BiruFang_FaceAcc_Physics
    simulationSpace=BaseBoneSpace
    alphaBoolBlend {blendOption=Linear}
    baseBoneRef {boneName=spine_05}
    componentAppliedLinearAccClamp=<3000 3000 3000>
    componentLinearAccScale=<0.35 0.2 0.2>
    componentLinearVelScale=<0.2 0.1 0.1>
    overrideWorldGravity {z=-980}
    simSpaceSettings {
     maxAngularAcceleration=1000
     maxAngularVelocity=3
     maxLinearAcceleration=700
     maxLinearVelocity=500
     velocityScaleZ=0
     worldAlpha=1
     externalLinearDragV=<0.2 0.2>
    }
   }
   n77 = Control Rig(Source=Component To Local(ComponentPose=@n76)) {
    alphaBoolBlend {blendOption=Linear}
    controlRigAssetReference {blueprintRigClass=&/M_MED_BiruFang_CtrlRig.M_MED_BiruFang_CtrlRig_C}
   }
   Output Pose(Result=@n77) {layerGroup=None}
  ]}
 EventGraph {nodes [
   Event BlueprintInitializeAnimation()
   > Set skeletal_mesh(skeletal_mesh:Skeletal Mesh Component Object Reference=GetPartSkeletalMeshComponent()) -> (Output_Get:Skeletal Mesh Component Object Reference)
   n1 = Event BlueprintUpdateAnimation()
   n2 = Get bIsJumping()
   n3 = Get bIsFalling()
   > Set AnimSpring(AnimSpring:double=SelectFloat(A=1750, B=2750, bPickA=OR Boolean(A=@n2, B=@n3))) -> (Output_Get:double)
   n4 = Get bIsJumping()
   n5 = Get bIsFalling()
   > Set AnimDynaAlpha(AnimDynaAlpha:double=SelectFloat(A=0.75, B=0.27, bPickA=OR Boolean(A=@n4, B=@n5))) -> (Output_Get:double)
   n6 = Get bIsCrouching()
   n7 = Get bIsFalling()
   n8 = Get bIsJumping()
   > Set isCoruchingJumpingFalling(isCoruchingJumpingFalling:double="To Float (Boolean)"(InBool=OR Boolean!(A=@n6, B=@n8, C:bool=@n7) -> (ReturnValue))) -> (Output_Get:double)
   > Set isCrouching(isCrouching:double="To Float (Boolean)"(InBool=@n6)) -> (Output_Get:double)
   > Set isNotCrouching(isNotCrouching:double="To Float (Boolean)"(InBool=NOT Boolean(A=@n6))) -> (Output_Get:double)
   > Set Is Playing Emote(Is Playing Emote:double="To Float (Boolean)"(InBool=Get bIsPlayingEmote())) -> (Output_Get:double)
   > Set IsDBNO(IsDBNO:double="To Float (Boolean)"(InBool=Get bIsDBNO())) -> (Output_Get:double)
   > Is Valid(InputObject=Get Player Pawn() -> (Player Pawn:Fort Player Pawn Athena Object Reference))
   > n9 = Sequence()
   > Set WPO Globals Function(
    Material=Get material_instance() -> (material_instance:Material Instance Dynamic Object Reference)
    Overlay Material=Get overlay_mid() -> (overlay_mid:Material Instance Dynamic Object Reference)
    Skeletal Mesh=Get skeletal_mesh() -> (skeletal_mesh:Skeletal Mesh Component Object Reference)
    "Local Noise Transform  Location"="Get Local Noise Transform  Location"() -> ("Local Noise Transform  Location":Vector)
    Local Noise Offset=Get Local Noise Offset()
   )
   > Set WPO Regional Function!(
    execute
    Material:Material Instance Dynamic Object Reference=Get material_instance()
    Overlay Material:Material Instance Dynamic Object Reference=Get overlay_mid()
    Skeletal Mesh:Skeletal Mesh Component Object Reference=Get skeletal_mesh()
    Velocity Transform:Name=spine_05
    Velocity Interp Speed:double=Get Velocity Interp Speed()
    Collision Transform:Name=spine_05
    Collision Transform Axis:EAxis Enum
    Invert Collision Transform Axis:bool=true
    Delta Time X:double=@n1.DeltaTimeX
    Previous Velocity Transform Location:"Vector (by ref)"=Get Previous Velocity Transform Location() -> (Previous Velocity Transform Location:Vector)
    Previous Velocity Transform Velocity:"Vector (by ref)"=Get Previous Velocity Transform Velocity() -> (Previous Velocity Transform Velocity:Vector)
    Region:int=1
   ) -> (then)
   n10 = Branch(execute=@n9.then_1, Condition=Get bIsSkyDiving())
   > n11 = Set WPO Pho Wind Intensity(WPO Pho Wind Intensity=1)
   SetScalarParameterValue(execute=[@n12.then, @n11.then], self=Get material_instance(), ParameterName=ClothSim_WindStrength, Value=Get WPO Pho Wind Intensity())
   n12 = Set WPO Pho Wind Intensity(execute=@n10.else, WPO Pho Wind Intensity=0)
   Event OnCharacterCustomizationLoaded()
   > n13 = Sequence()
   n14 = TryGetPawnOwner()
   > Is Valid(InputObject=@n14)
   > n15 = Cast To FortPlayerPawnAthena(Object=@n14)
   > Set Player Pawn(Player Pawn:Fort Player Pawn Athena Object Reference=@n15) -> (Output_Get:Fort Player Pawn Athena Object Reference)
   n16 = Get skeletal_mesh()
   n17 = Get Dynamic Material Instance(execute=@n13.then_1, SkeletalMeshComponent=@n16)
   > Set material_instance(material_instance:Material Instance Dynamic Object Reference=@n17) -> (Output_Get:Material Instance Dynamic Object Reference)
   > n18 = GetOrCreateDynamicOverlayMaterial(MeshComponent=@n16, MaterialName=None)
   > Set overlay_mid(overlay_mid:Material Instance Dynamic Object Reference=@n18) -> (Output_Get:Material Instance Dynamic Object Reference)
  ]}
]
variables [
 AnimDynaAlpha {type=number}
 AnimSpring {type=number}
 Is Playing Emote {type=number}
 isCoruchingJumpingFalling {type=number}
 isCrouching {type=number}
 IsDBNO {type=number}
 isNotCrouching {type=number}
 Local Noise Offset {type=Vector}
 "Local Noise Transform  Location" {type=Vector}
 material_instance {type=MaterialInstanceDynamic}
 overlay_mid {type=MaterialInstanceDynamic}
 Player Pawn {type=FortPlayerPawnAthena}
 Previous Velocity Transform Location {type=Vector}
 Previous Velocity Transform Velocity {type=Vector}
 skeletal_mesh {type=SkeletalMeshComponent}
 Velocity Interp Speed {type=number}
 WPO Pho Wind Intensity {type=number}
]
